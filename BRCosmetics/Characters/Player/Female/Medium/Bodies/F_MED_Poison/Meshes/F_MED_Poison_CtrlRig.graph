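# Control Rig for the F_MED_Poison female body cosmetic that nudges back-belt bones as the thighs move.
# A sequence runs two spherical pose readers, thigh_l weighting an offset on dyn_belt_bk_03 and thigh_r on dyn_belt_bk_06.
# A shared function adds translation, Euler rotation and scale deltas to a bone's current local transform by weight.
# Two further calls, including a skirt correction, sit outside the execution chain; left-reader debug drawing is on.
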ControlRigBlueprint /BRCosmetics/Characters/Player/Female/Medium/Bodies/F_MED_Poison/Meshes/F_MED_Poison_CtrlRig
graphs [
 ContainedGraph {nodes [
   n1 = RigVMFunctionEntryNode()
   RigUnit_GetTransform = &RigUnit_GetTransform(Space=LocalSpace, bInitial=false, Item=@n1.Joint)
   n2 = &RigVMFunction_MathQuaternionToEuler(RotationOrder=ZYX, Value=@RigUnit_GetTransform.Transform.Rotation)
   > RigUnit_SetTransform = &RigUnit_SetTransform(
    Space=LocalSpace
    Value=(Rotation=(X=0.0132,Y=0.4099,Z=-0.0222,W=0.9118),Translation=(X=23.6045,Y=-6.37,Z=140.1115),Scale3D=(X=1,Y=1,Z=1))
    bInitial=false
    Item=@n1.Joint
    Weight=@n1.Weight
    Value.Translation.Z=&RigVMFunction_MathFloatAdd(B=@n1.Translate Z, A=@RigUnit_GetTransform.Transform.Translation.Z) {template="Add::Execute(in A,in B,out Result)"}.Result
    Value.Scale3D.Z=&RigVMFunction_MathFloatAdd(B=@n1.Scale Z, A=@RigUnit_GetTransform.Transform.Scale3D.Z) {template="Add::Execute(in A,in B,out Result)"}.Result
    Value.Translation.Y=&RigVMFunction_MathFloatAdd(B=@n1.Translate Y, A=@RigUnit_GetTransform.Transform.Translation.Y) {template="Add::Execute(in A,in B,out Result)"}.Result
    Value.Scale3D.Y=&RigVMFunction_MathFloatAdd(B=@n1.Scale Y, A=@RigUnit_GetTransform.Transform.Scale3D.Y) {template="Add::Execute(in A,in B,out Result)"}.Result
    Value.Translation.X=&RigVMFunction_MathFloatAdd(B=@n1.Translate X, A=@RigUnit_GetTransform.Transform.Translation.X) {template="Add::Execute(in A,in B,out Result)"}.Result
    Value.Scale3D.X=&RigVMFunction_MathFloatAdd(B=@n1.Scale X, A=@RigUnit_GetTransform.Transform.Scale3D.X) {template="Add::Execute(in A,in B,out Result)"}.Result
    Value.Rotation=&RigVMFunction_MathQuaternionFromEuler(
     Euler=(X=0,Y=0,Z=0)
     Euler.X=&RigVMFunction_MathFloatAdd(B=@n1.Rotate X, A=@n2.Result.X) {template="Add::Execute(in A,in B,out Result)"}.Result
     Euler.Y=&RigVMFunction_MathFloatAdd(B=@n1.Rotate Y, A=@n2.Result.Y) {template="Add::Execute(in A,in B,out Result)"}.Result
     Euler.Z=&RigVMFunction_MathFloatAdd(B=@n1.Rotate Z, A=@n2.Result.Z) {template="Add::Execute(in A,in B,out Result)"}.Result
    ).Result
   )
   > RigVMFunctionReturnNode()
  ]}
 RigVMFunctionLibrary {nodes [Set Transforms for Bone = RigVMCollapseNode(
    ExecuteContext=()
    Joint=()
    Rotate X=0
    Rotate Y=0
    Rotate Z=0
    Scale X=0
    Scale Y=0
    Scale Z=0
    Translate X=0
    Translate Y=0
    Translate Z=0
    Weight=1
   )]}
 RigVMModel {nodes [
   RigUnit_BeginExecution = &RigUnit_BeginExecution()
   n1 = /Script/ControlRig.RigUnit_SequenceExecution(ExecuteContext=@RigUnit_BeginExecution.ExecutePin) {is=struct RigVMUnitNode}
   n2 = &RigUnit_SphericalPoseReader(
    ActiveRegionScaleFactors=(PositiveWidth=3.5,NegativeWidth=1,PositiveHeight=3.5,NegativeHeight=4)
    ActiveRegionSize=0.1
    Debug=(bDrawDebug=True,DebugScale=25,DebugSegments=20,DebugThickness=0.25)
    DriverAxis=(X=-1,Y=0,Z=0)
    DriverItem=(Type=Bone,Name="thigh_l")
    FalloffRegionScaleFactors=(PositiveWidth=1,NegativeWidth=0.9,PositiveHeight=1,NegativeHeight=1)
    FalloffSize=0.45
    FlipHeightScaling=true
    FlipWidthScaling=true
    OptionalParentItem=(Type=Bone,Name="None")
    RotationOffset=(X=90,Y=180,Z=0)
    ExecutePin=@n1.A
   )
   > Set Transforms for Bone_1_1_1_4_1_1_1 = RigVMFunctionReferenceNode(
    Joint=(Type=Bone,Name="dyn_belt_bk_03")
    Rotate X=0
    Rotate Y=0
    Rotate Z=0
    Scale X=0
    Scale Y=0
    Scale Z=0
    Translate X=0
    Translate Y=0
    Translate Z=10
    Weight=@n2.OutputParam
   )
   n3 = &RigUnit_SphericalPoseReader(
    ActiveRegionScaleFactors=(PositiveWidth=3.5,NegativeWidth=0.5,PositiveHeight=3.5,NegativeHeight=4)
    ActiveRegionSize=0.1
    Debug=(bDrawDebug=false,DebugScale=25,DebugSegments=20,DebugThickness=0.25)
    DriverAxis=(X=1,Y=0,Z=0)
    DriverItem=(Type=Bone,Name="thigh_r")
    FalloffRegionScaleFactors=(PositiveWidth=1,NegativeWidth=1.1,PositiveHeight=1,NegativeHeight=1)
    FalloffSize=0.45
    FlipHeightScaling=true
    FlipWidthScaling=true
    OptionalParentItem=(Type=Bone,Name="None")
    RotationOffset=(X=-90,Y=0,Z=0)
    ExecutePin=@n1.B
   )
   > Set Transforms for Bone_1_1_1_4_1_1_1_1 = RigVMFunctionReferenceNode(
    Joint=(Type=Bone,Name="dyn_belt_bk_06")
    Rotate X=0
    Rotate Y=0
    Rotate Z=0
    Scale X=0
    Scale Y=0
    Scale Z=0
    Translate X=0
    Translate Y=13
    Translate Z=0
    Weight=@n3.OutputParam
   )
   Set Transforms for Bone_1_1_1_4_1_1_1_2 = RigVMFunctionReferenceNode(
    ExecuteContext=()
    Joint=(Type=Bone,Name="dyn_belt_bk_03")
    Rotate X=0
    Rotate Y=0
    Rotate Z=0
    Scale X=0
    Scale Y=0
    Scale Z=0
    Translate X=0
    Translate Y=0
    Translate Z=0
    Weight=1
   )
   Set Transforms for Bone_1_1_1_4_3 = RigVMFunctionReferenceNode(
    ExecuteContext=()
    Joint=(Type=Bone,Name="dyn_skirt_fr_mid_r_02")
    Rotate X=0
    Rotate Y=30
    Rotate Z=-90
    Scale X=0
    Scale Y=0
    Scale Z=0
    Translate X=-12
    Translate Y=0
    Translate Z=0
    Weight=@n2.OutputParam
   )
  ]}
]
hierarchy [
 root {loc=<0 0 0>, rot=<0 0 0 1>}
 attach {parent=root, loc=<0 0 0>, rot=<0 0 0 1>}
 pelvis {parent=root, loc=<0 -0.5301 98.0571>, rot=<0.0127 0.707 -0.0127 -0.707>}
 spine_01 {parent=pelvis, loc=<6.1208 -0.9328 0>, rot=<0 0 0.034 -0.9994>}
 spine_02 {parent=spine_01, loc=<6.5996 -0.2095 0>, rot=<0 0 0.0238 -0.9997>}
 spine_03 {parent=spine_02, loc=<6.4914 -0.1237 0>, rot=<0 0 0.0149 -0.9999>}
 spine_04 {parent=spine_03, loc=<6.9598 0.0794 0>, rot=<0 0 -0.0132 -0.9999>}
 spine_05 {parent=spine_04, loc=<6.9987 0 0>, rot=<0 0 -0.0015 -1>}
 clavicle_l {parent=spine_05, loc=<5.4831 -0.1053 -3.2946>, rot=<0.0465 -0.7289 -0.009 -0.6829>}
 upperarm_l {parent=clavicle_l, loc=<9.8658 -1.2136 -0.4105>, rot=<0.0164 -0.3245 -0.0269 -0.9454>}
 lowerarm_l {parent=upperarm_l, loc=<26.465 -0.8345 0.259>, rot=<0.0036 0.0161 -0.2298 -0.9731>}
 hand_l {parent=lowerarm_l, loc=<22.7355 -0.3203 -0.6105>, rot=<-0.6784 -0.0223 0.0735 -0.7307>}
 index_metacarpal_l {parent=hand_l, loc=<1.6535 0.0183 -2.96>, rot=<0.0158 -0.0124 -0.0173 -0.9996>}
 index_01_l {parent=index_metacarpal_l, loc=<8.3008 0.0357 -0.0524>, rot=<-0.0044 0.0485 0.0924 -0.9945>}
 index_02_l {parent=index_01_l, loc=<3.3587 -0.3004 0.0452>, rot=<-0.002 0.0027 0.3225 -0.9466>}
 index_03_l {parent=index_02_l, loc=<2.5578 -0.123 -0.0252>, rot=<-0.002 0.0067 0.2558 -0.9667>}
 middle_metacarpal_l {parent=hand_l, loc=<2.1856 -0.1725 -0.5984>, rot=<0.006 0.0168 -0.0329 -0.9993>}
 middle_01_l {parent=middle_metacarpal_l, loc=<7.3275 -0.2447 0.0129>, rot=<-0.0208 0.0042 0.0858 -0.9961>}
 middle_02_l {parent=middle_01_l, loc=<4.0123 0.1401 0.1175>, rot=<-0.0218 0.0258 0.3512 -0.9357>}
 middle_03_l {parent=middle_02_l, loc=<2.6077 0.0534 0.1373>, rot=<0.0082 0.0188 0.1708 -0.9851>}
 pinky_metacarpal_l {parent=hand_l, loc=<1.7616 -0.9753 3.3814>, rot=<-0.0397 0.0814 -0.069 -0.9935>}
 pinky_01_l {parent=pinky_metacarpal_l, loc=<7.1709 -0.4831 -0.1094>, rot=<-0.0654 -0.0035 0.2242 -0.9723>}
 pinky_02_l {parent=pinky_01_l, loc=<2.8274 0.0893 -0.1043>, rot=<0.0037 0.0011 0.2854 -0.9584>}
 pinky_03_l {parent=pinky_02_l, loc=<2.2046 0.0307 -0.0078>, rot=<-0.0002 0.0098 0.1445 -0.9895>}
 ring_metacarpal_l {parent=hand_l, loc=<2.2074 -0.3625 1.3638>, rot=<-0.0284 0.0722 -0.0458 -0.9959>}
 ring_01_l {parent=ring_metacarpal_l, loc=<7.0295 -0.6197 -0.1446>, rot=<-0.0489 -0.0252 0.1892 -0.9804>}
 ring_02_l {parent=ring_01_l, loc=<3.6174 0.4647 -0.1027>, rot=<0 0 0.2806 -0.9598>}
 ring_03_l {parent=ring_02_l, loc=<2.4152 0.114 0.0199>, rot=<0 0 0.1189 -0.9929>}
 thumb_01_l {parent=hand_l, loc=<1.3118 -0.3191 -2.459>, rot=<0.6546 -0.4042 0.0664 -0.6354>}
 thumb_02_l {parent=thumb_01_l, loc=<5.9336 -0.0448 0.0401>, rot=<-0.0037 0.0213 0.298 -0.9543>}
 thumb_03_l {parent=thumb_02_l, loc=<2.2223 0.3682 -0.0907>, rot=<0.0118 -0.013 0.0066 -0.9998>}
 weapon_l {parent=hand_l, loc=<-0.1974 -3.4111 0.3811>, rot=<0 0 0 -1>}
 lowerarm_twist_01_l {parent=lowerarm_l, loc=<18.316 0 0>, rot=<0 0 0 1>}
 lowerarm_twist_02_l {parent=lowerarm_l, loc=<10.6707 0 0>, rot=<0 0 0 1>}
 dyn_wrist_strap_l_MSTR {parent=lowerarm_twist_02_l, loc=<2.6828 -4.01 0.9181>, rot=<0.0722 0.3332 0.2199 -0.914>}
 dyn_wrist_strap_l {parent=dyn_wrist_strap_l_MSTR, loc=<0 0 0>, rot=<0 0 0 1>}
 upperarm_twist_01_l {parent=upperarm_l, loc=<7.2135 -0.1134 0>, rot=<0 0 0 1>}
 upperarm_twist_02_l {parent=upperarm_l, loc=<19.7998 -0.1134 0>, rot=<0 0 0 1>}
 clavicle_r {parent=spine_05, loc=<5.4831 -0.1053 3.2946>, rot=<0.7289 0.0465 -0.6829 0.009>}
 upperarm_r {parent=clavicle_r, loc=<-9.8658 1.2136 0.4104>, rot=<0.0164 -0.3245 -0.0269 -0.9454>}
 lowerarm_r {parent=upperarm_r, loc=<-26.4651 0.8345 -0.259>, rot=<0.0036 0.0161 -0.2298 -0.9731>}
 hand_r {parent=lowerarm_r, loc=<-22.7356 0.3203 0.6106>, rot=<-0.6784 -0.0223 0.0735 -0.7307>}
 index_metacarpal_r {parent=hand_r, loc=<-1.6534 -0.0181 2.9601>, rot=<0.0158 -0.0124 -0.0173 -0.9996>}
 index_01_r {parent=index_metacarpal_r, loc=<-8.3008 -0.0357 0.0524>, rot=<-0.0044 0.0485 0.0924 -0.9945>}
 index_02_r {parent=index_01_r, loc=<-3.3589 0.3003 -0.0453>, rot=<-0.002 0.0027 0.3225 -0.9466>}
 index_03_r {parent=index_02_r, loc=<-2.5578 0.123 0.0252>, rot=<-0.002 0.0067 0.2558 -0.9667>}
 middle_metacarpal_r {parent=hand_r, loc=<-2.1856 0.1726 0.5984>, rot=<0.006 0.0168 -0.0329 -0.9993>}
 middle_01_r {parent=middle_metacarpal_r, loc=<-7.3273 0.2447 -0.0129>, rot=<-0.0208 0.0042 0.0858 -0.9961>}
 middle_02_r {parent=middle_01_r, loc=<-4.0124 -0.14 -0.1175>, rot=<-0.0218 0.0258 0.3512 -0.9357>}
 middle_03_r {parent=middle_02_r, loc=<-2.6077 -0.0533 -0.1373>, rot=<0.0082 0.0188 0.1708 -0.9851>}
 pinky_metacarpal_r {parent=hand_r, loc=<-1.7617 0.9753 -3.3814>, rot=<-0.0397 0.0814 -0.069 -0.9935>}
 pinky_01_r {parent=pinky_metacarpal_r, loc=<-7.1707 0.4832 0.1094>, rot=<-0.0654 -0.0035 0.2242 -0.9723>}
 pinky_02_r {parent=pinky_01_r, loc=<-2.8274 -0.0892 0.1043>, rot=<0.0037 0.0011 0.2854 -0.9584>}
 pinky_03_r {parent=pinky_02_r, loc=<-2.2045 -0.031 0.0078>, rot=<-0.0002 0.0098 0.1445 -0.9895>}
 ring_metacarpal_r {parent=hand_r, loc=<-2.2073 0.3627 -1.3638>, rot=<-0.0284 0.0722 -0.0458 -0.9959>}
 ring_01_r {parent=ring_metacarpal_r, loc=<-7.0297 0.6197 0.1446>, rot=<-0.0489 -0.0252 0.1892 -0.9804>}
 ring_02_r {parent=ring_01_r, loc=<-3.6171 -0.4648 0.1027>, rot=<0 0 0.2806 -0.9598>}
 ring_03_r {parent=ring_02_r, loc=<-2.4153 -0.1141 -0.0199>, rot=<0 0 0.1189 -0.9929>}
 thumb_01_r {parent=hand_r, loc=<-1.3119 0.3192 2.459>, rot=<0.6546 -0.4042 0.0664 -0.6354>}
 thumb_02_r {parent=thumb_01_r, loc=<-5.9335 0.0447 -0.0401>, rot=<-0.0037 0.0213 0.298 -0.9543>}
 thumb_03_r {parent=thumb_02_r, loc=<-2.2223 -0.3682 0.0907>, rot=<0.0118 -0.013 0.0066 -0.9998>}
 weapon_r {parent=hand_r, loc=<0.1975 3.4112 -0.3811>, rot=<0 0 0 -1>}
 lowerarm_twist_01_r {parent=lowerarm_r, loc=<-18.3161 0 0>, rot=<0 0 0 1>}
 lowerarm_twist_02_r {parent=lowerarm_r, loc=<-10.6707 0 0>, rot=<0 0 0 1>}
 dyn_wrist_strap_l_MSTR1 {parent=lowerarm_twist_02_r, loc=<-2.6825 4.0099 -0.9183>, rot=<-0.4756 0.4929 -0.5226 -0.5077>}
 dyn_wrist_strap_l1 {parent=dyn_wrist_strap_l_MSTR1, loc=<0 0 0>, rot=<0 0 0 1>}
 upperarm_twist_01_r {parent=upperarm_r, loc=<-7.2135 0.1134 0>, rot=<0 0 0 1>}
 upperarm_twist_02_r {parent=upperarm_r, loc=<-19.7998 0.1134 0>, rot=<0 0 -0.0037 -1>}
 neck_01 {parent=spine_05, loc=<11.3138 0.1529 0>, rot=<0 0 -0.2003 -0.9797>}
 neck_02 {parent=neck_01, loc=<4.5627 0 0>, rot=<0 0 0.0504 -0.9987>}
 head {parent=neck_02, loc=<5.3914 0 0>, rot=<0 0 0.115 -0.9934>}
 dyn_neck_01_LOD1_bk {parent=neck_01, loc=<0 0 0>, rot=<0 0 0 1>}
 dyn_clip_back_MSTR {parent=spine_05, loc=<5.8125 -7.5807 0.0002>, rot=<0.0342 -0.7099 0.7032 -0.0224>}
 dyn_clip_back_01 {parent=dyn_clip_back_MSTR, loc=<0 0 0>, rot=<-0.0069 0.2442 -0.0067 -0.9697>}
 dyn_clip_back_02 {parent=dyn_clip_back_01, loc=<1.5295 0 0>, rot=<0 -0.1899 0 -0.9818>}
 dyn_collar_00 {parent=spine_05, loc=<15.4661 -6.1347 -0.0095>, rot=<0.0284 -0.7065 -0.0284 -0.7065>}
 dyn_collar_01 {parent=dyn_collar_00, loc=<-0.0095 0.2822 5.5804>, rot=<0 0 0 1>}
 dyn_clip_front_MSTR {parent=spine_03, loc=<-2.0058 10.4414 -0.0337>, rot=<0.0166 -0.7026 -0.0459 -0.7099>}
 dyn_clip_front_01 {parent=dyn_clip_front_MSTR, loc=<0 0 0>, rot=<0 0 0 1>}
 dyn_clip_front_02 {parent=dyn_clip_front_01, loc=<-0.017 -0.0273 -1.3106>, rot=<0 0 0 1>}
 thigh_l {parent=pelvis, loc=<-1.0855 -0.2109 -8.7053>, rot=<0.0009 0.0355 0.0155 -0.9993>}
 calf_l {parent=thigh_l, loc=<-42.5284 -0.0019 -0.7297>, rot=<0.0006 -0.0161 -0.0391 -0.9991>}
 calf_twist_01_l {parent=calf_l, loc=<-35.8483 -1.9197 -0.5794>, rot=<0.0237 0.0095 -0.0035 -0.9997>}
 calf_twist_02_l {parent=calf_l, loc=<-19.1546 -1.7943 -0.2638>, rot=<0.0237 0.0095 -0.0035 -0.9997>}
 dyn_pants_l_ft_00 {parent=calf_twist_02_l, loc=<5.3876 1.6862 -0.3839>, rot=<0.0425 0.6885 0.7235 0.028>}
 dyn_pants_l_ft_01 {parent=dyn_pants_l_ft_00, loc=<6.7273 0.0035 0.0144>, rot=<0 -0.0049 0.0012 -1>}
 dyn_pants_l_in_00 {parent=calf_twist_02_l, loc=<4.3907 -3.7235 3.4394>, rot=<0.0754 -0.0289 0.9963 0.0279>}
 dyn_pants_l_in_01 {parent=dyn_pants_l_in_00, loc=<5.7738 0.0457 0.0483>, rot=<-0.0003 -0.0163 0.0155 -0.9997>}
 dyn_pants_l_ex_00 {parent=calf_twist_02_l, loc=<4.6557 -3.3435 -4.9423>, rot=<-0.0324 -0.9966 -0.0199 -0.073>}
 dyn_pants_l_ex_01 {parent=dyn_pants_l_ex_00, loc=<6.0499 -0.0428 -0.0922>, rot=<-0.0005 0.0312 -0.0145 -0.9994>}
 dyn_pants_l_bk_00 {parent=calf_twist_02_l, loc=<4.5046 -6.9848 -0.7958>, rot=<0.052 -0.7826 0.6195 -0.0328>}
 dyn_pants_l_bk_01 {parent=dyn_pants_l_bk_00, loc=<6.0575 0 0>, rot=<0.0712 -0.0098 -0.0148 -0.9973>}
 foot_l {parent=calf_l, loc=<-46.7402 -2.5014 -0.592>, rot=<0.0427 -0.0238 0.0353 -0.9982>}
 ball_l {parent=foot_l, loc=<-3.5658 10.8066 -0.1544>, rot=<0.0001 -0.0001 -0.7063 -0.7079>}
 dyn_laces_l {parent=foot_l, loc=<3.6151 -6.3888 0.2901>, rot=<-0.0357 -0.7045 -0.0284 -0.7082>}
 dyn_shoe_strap_l {parent=foot_l, loc=<0.8119 6.5819 -1.845>, rot=<-0.0341 0.708 0.0414 -0.7042>}
 thigh_twist_01_l {parent=thigh_l, loc=<-27.952 0 0>, rot=<0.0013 0 0 -1>}
 dyn_zipper_l_MSTR {parent=thigh_twist_01_l, loc=<-4.7417 3.9491 4.186>, rot=<0.8993 0.0145 -0.0307 0.436>}
 dyn_zipper_l {parent=dyn_zipper_l_MSTR, loc=<0 0 0>, rot=<0 0 0 1>}
 thigh_r {parent=pelvis, loc=<-1.0853 -0.2109 8.7053>, rot=<0.032 -0.0005 0.9993 0.0194>}
 calf_r {parent=thigh_r, loc=<42.5286 0.0019 0.7297>, rot=<0.0006 -0.0161 -0.0391 -0.9991>}
 calf_twist_01_r {parent=calf_r, loc=<35.8485 1.9197 0.5794>, rot=<0.0237 0.0095 -0.0035 -0.9997>}
 calf_twist_02_r {parent=calf_r, loc=<19.1546 1.7943 0.2638>, rot=<0.0237 0.0095 -0.0035 -0.9997>}
 dyn_pants_r_ft_00 {parent=calf_twist_02_r, loc=<-5.3635 -1.2628 0.6584>, rot=<-0.7233 0.0292 0.036 -0.689>}
 dyn_pants_r_ft_01 {parent=dyn_pants_r_ft_00, loc=<6.6448 -0.0198 0.015>, rot=<0 -0.0049 -0.0064 -1>}
 dyn_pants_r_in_00 {parent=calf_twist_02_r, loc=<-4.3792 4.1504 -3.1542>, rot=<0.9967 -0.0241 -0.0719 -0.0282>}
 dyn_pants_r_in_01 {parent=dyn_pants_r_in_00, loc=<5.6904 -0.0478 0.0504>, rot=<0.0003 -0.0163 -0.0155 -0.9997>}
 dyn_pants_r_ex_00 {parent=calf_twist_02_r, loc=<-4.7019 3.7769 5.2307>, rot=<-0.0198 0.0766 0.0286 -0.9965>}
 dyn_pants_r_ex_01 {parent=dyn_pants_r_ex_00, loc=<5.9667 0.0448 -0.0963>, rot=<0.0005 0.0312 0.0145 -0.9994>}
 dyn_pants_r_bk_00 {parent=calf_twist_02_r, loc=<-4.5474 7.6521 1.0903>, rot=<0.6877 0.0483 -0.0513 -0.7226>}
 dyn_pants_r_bk_01 {parent=dyn_pants_r_bk_00, loc=<5.9146 0.0296 -0.0629>, rot=<0.0012 0.0203 0.0096 -0.9997>}
 foot_r {parent=calf_r, loc=<46.7403 2.5014 0.592>, rot=<0.0427 -0.0238 0.0353 -0.9982>}
 ball_r {parent=foot_r, loc=<3.5658 -10.8066 0.1544>, rot=<0.0001 -0.0001 -0.7063 -0.7079>}
 dyn_laces_r {parent=foot_r, loc=<-3.6153 6.3888 -0.2901>, rot=<0.4979 -0.4997 0.5056 -0.4968>}
 dyn_shoe_strap_r {parent=foot_r, loc=<-0.8121 -6.5819 1.845>, rot=<0.7082 0.0284 -0.7045 -0.0357>}
 thigh_twist_01_r {parent=thigh_r, loc=<27.952 0 0>, rot=<0.0013 0 0 -1>}
 dyn_zipper_r_MSTR {parent=thigh_twist_01_r, loc=<4.8984 -3.7279 -3.9164>, rot=<-0.6816 0.0006 0.7317 0.0028>}
 dyn_zipper_r {parent=dyn_zipper_r_MSTR, loc=<0 0 0>, rot=<0 0 0 1>}
 dyn_skirt_fr_mid_r_02 {parent=thigh_twist_01_r, loc=<-14.4944 -7.5113 -0.8873>, rot=<-0.7079 -0.0015 0.0072 -0.7062>}
 dyn_skirt_side_03_r {parent=thigh_twist_01_r, loc=<-14.7157 2.78 7.7227>, rot=<0.0001 0.0341 -0.043 -0.9985>}
 dyn_skirt_03_r {parent=thigh_twist_01_r, loc=<-15.0627 -4.9609 6.1674>, rot=<-0.2715 0.0335 -0.0238 -0.9616>}
 dyn_belt_bk_00 {parent=pelvis, loc=<1.8517 -11.1172 6.3087>, rot=<-0.3295 -0.5569 0.6697 -0.3644>}
 dyn_belt_bk_01 {parent=dyn_belt_bk_00, loc=<9.9638 0 0>, rot=<0.0587 -0.103 -0.0199 -0.9928>}
 dyn_belt_bk_02 {parent=dyn_belt_bk_01, loc=<5.3141 0 0>, rot=<0.1157 -0.2 0.0258 -0.9726>}
 dyn_belt_bk_03 {parent=dyn_belt_bk_02, loc=<6.1418 0 0>, rot=<0.0648 -0.1319 -0.0935 -0.9847>}
 dyn_belt_bk_04 {parent=dyn_belt_bk_03, loc=<6.2049 0 0>, rot=<0.0462 -0.1927 -0.1237 -0.9723>}
 dyn_belt_bk_05 {parent=dyn_belt_bk_04, loc=<3.0362 0 0>, rot=<0.0094 -0.1669 -0.0513 -0.9846>}
 dyn_belt_bk_06 {parent=dyn_belt_bk_05, loc=<4.063 0 0>, rot=<-0.0027 -0.2045 -0.0174 -0.9787>}
 dyn_mask_04 {parent=dyn_belt_bk_04, loc=<5.6652 2.482 -0.045>, rot=<-0.3535 0.401 0.4869 -0.6908>}
 dyn_mask_05 {parent=dyn_mask_04, loc=<5.0347 0 0>, rot=<-0.0516 0.0136 -0.024 -0.9983>}
 dyn_mask_06 {parent=dyn_mask_05, loc=<3.0512 0 0>, rot=<0.1186 -0.0232 0.0882 -0.9887>}
 dyn_mask_07 {parent=dyn_mask_06, loc=<14.4238 0 0>, rot=<0.4126 -0.4581 -0.5269 -0.5851>}
 dyn_mask_08 {parent=dyn_mask_07, loc=<5.8978 0 0>, rot=<0 0 0.1598 -0.9871>}
 dyn_grenade {parent=dyn_belt_bk_00, loc=<7.1856 -0.089 0.512>, rot=<0.3834 0.6725 0.2188 -0.594>}
 dyn_grenade_pin {parent=dyn_grenade, loc=<3.473 4.2989 6.5235>, rot=<-0.6184 0.0441 0.4061 -0.6714>}
 dyn_mask_MSTR {parent=pelvis, loc=<3.76 -0.3966 -12.8836>, rot=<-0.0127 -0.707 0.0127 -0.707>}
 dyn_mask_01 {parent=dyn_mask_MSTR, loc=<0 0 0>, rot=<0.565 -0.4252 -0.4252 -0.565>}
 dyn_mask_02 {parent=dyn_mask_01, loc=<6.4504 0 0>, rot=<0 0 0.038 -0.9993>}
 dyn_belt_ft_MSTR {parent=pelvis, loc=<4.1902 -0.2376 14.0716>, rot=<-0.0127 -0.707 0.0127 -0.707>}
 dyn_belt_ft_base_00 {parent=dyn_belt_ft_MSTR, loc=<0 0 0>, rot=<0.6624 0.2475 -0.5253 -0.4733>}
 dyn_belt_ft_base_01 {parent=dyn_belt_ft_base_00, loc=<5.0776 0 0>, rot=<-0.1304 -0.2114 -0.0707 -0.9661>}
 dyn_belt_ft_00 {parent=dyn_belt_ft_base_01, loc=<5.9038 0 0>, rot=<-0.1118 -0.1681 0.038 -0.9787>}
 dyn_belt_ft_01 {parent=dyn_belt_ft_00, loc=<6.8104 0 0>, rot=<-0.0402 -0.072 0.0389 -0.9958>}
 dyn_belt_ft_02 {parent=dyn_belt_ft_01, loc=<7.5837 0 0>, rot=<-0.0862 -0.2022 0.0652 -0.9734>}
 dyn_belt_ft_03 {parent=dyn_belt_ft_02, loc=<6.5359 0 0>, rot=<-0.0098 -0.0496 0.1414 -0.9887>}
 dyn_belt_ft_04 {parent=dyn_belt_ft_03, loc=<5.8497 0 0>, rot=<-0.0161 -0.3764 0.0087 -0.9263>}
 dyn_belt_ft_05 {parent=dyn_belt_ft_04, loc=<3.6994 0 0>, rot=<-0.002 -0.1813 0.022 -0.9832>}
 dyn_belt_ft_06 {parent=dyn_belt_ft_05, loc=<4.0901 0 0>, rot=<0.0118 -0.1764 0.0544 -0.9827>}
 dyn_belt_00 {parent=dyn_belt_ft_00, loc=<1.7773 0.4211 1.0295>, rot=<0.2145 -0.0522 0.8966 0.3838>}
 dyn_belt_01 {parent=dyn_belt_00, loc=<3.1217 0 0>, rot=<0.0567 0.0121 0.3207 -0.9454>}
 dyn_belt_02 {parent=dyn_belt_01, loc=<4.6139 0 0>, rot=<0.0225 -0.0071 0.1233 -0.9921>}
 dyn_belt_03 {parent=dyn_belt_02, loc=<4.4439 0 0>, rot=<0.0179 -0.0054 0.1056 -0.9942>}
 dyn_belt_04 {parent=dyn_belt_03, loc=<8.156 0 0>, rot=<0.01 -0.0282 0.0741 -0.9968>}
 dyn_belt_05 {parent=dyn_belt_04, loc=<8.5403 0 0>, rot=<0.0019 -0.0295 0.0244 -0.9993>}
 dyn_belt_06 {parent=dyn_belt_05, loc=<8.2718 0 0>, rot=<0.0007 -0.0074 0.0164 -0.9998>}
 dyn_belt_07 {parent=dyn_belt_06, loc=<6.6398 0 0>, rot=<-0.503 0.4803 0.497 -0.5189>}
 dyn_skirt_fr_01_l_MSTR {parent=pelvis, loc=<0.3698 7.899 0.4402>, rot=<-0.0496 -0.7623 -0.6379 -0.0976>}
 dyn_skirt_fr_01_l {parent=dyn_skirt_fr_01_l_MSTR, loc=<0.2001 -0.0008 0.0001>, rot=<0.0879 -0.0583 -0.0244 -0.9941>}
 dyn_skirt_fr_02_l {parent=dyn_skirt_fr_01_l, loc=<5.4622 0 0>, rot=<-0.0004 -0.027 -0.0136 -0.9995>}
 dyn_skirt_fr_03_l {parent=dyn_skirt_fr_02_l, loc=<9.4106 0 0>, rot=<0 0 0 1>}
 dyn_skirt_fr_04_l {parent=dyn_skirt_fr_03_l, loc=<8.4441 0 0>, rot=<0 0 0 1>}
 dyn_skirt_bk_01_l_MSTR {parent=pelvis, loc=<1.3731 -10.6104 0.2694>, rot=<-0.0173 -0.6952 0.7184 -0.018>}
 dyn_skirt_bk_01_l {parent=dyn_skirt_bk_01_l_MSTR, loc=<0.0028 -0.0401 0.0087>, rot=<0.1357 0.0401 0.0146 -0.9898>}
 dyn_skirt_bk_02_l {parent=dyn_skirt_bk_01_l, loc=<8.7894 0 0>, rot=<-0.0858 -0.0039 0.0103 -0.9963>}
 dyn_skirt_bk_03_l {parent=dyn_skirt_bk_02_l, loc=<6.5732 0 0>, rot=<0.0068 -0.0047 -0.0008 -1>}
 dyn_skirt_bk_04_l {parent=dyn_skirt_bk_03_l, loc=<7.4421 0 0>, rot=<0 -0.0532 0 -0.9986>}
 dyn_skirt_bk_01_r_MSTR {parent=pelvis, loc=<2.0651 -7.6962 10.4195>, rot=<0.018 -0.7184 0.6952 0.0173>}
 dyn_skirt_bk_01_r {parent=dyn_skirt_bk_01_r_MSTR, loc=<0.0067 -0.0101 0.0138>, rot=<-0.3501 0.0894 0.0256 -0.9321>}
 dyn_skirt_bk_02_r {parent=dyn_skirt_bk_01_r, loc=<7.4888 0 0>, rot=<0.0168 -0.0388 -0.0035 -0.9991>}
 dyn_skirt_bk_03_r {parent=dyn_skirt_bk_02_r, loc=<8.8841 0 0>, rot=<0.0602 0.0172 -0.0113 -0.998>}
 dyn_skirt_bk_04_r {parent=dyn_skirt_bk_03_r, loc=<7.7391 0 0>, rot=<0 -0.1015 0 -0.9948>}
 dyn_skirt_01_r_side_MSTR {parent=pelvis, loc=<1.6532 -2.3391 13.9639>, rot=<0.1535 0.0166 0.988 -0.0057>}
 dyn_skirt_side_01_r {parent=dyn_skirt_01_r_side_MSTR, loc=<0.1995 0 0.0002>, rot=<0.0217 -0.026 -0.0002 -0.9994>}
 dyn_skirt_01_r_MSTR {parent=pelvis, loc=<-0.6428 4.3733 12.6515>, rot=<0.1535 0.0166 0.988 -0.0057>}
 dyn_skirt_01_r {parent=dyn_skirt_01_r_MSTR, loc=<0.1995 0 0.0002>, rot=<0.0231 -0.0398 0.0502 -0.9977>}
 dyn_skirt_fr_mid_r_MSTR {parent=pelvis, loc=<-5.425 7.1038 7.8464>, rot=<0.1396 0.43 0.8914 0.0323>}
 ik_foot_root {parent=root, loc=<0 0 0>, rot=<0 0 0 1>}
 ik_foot_l {parent=ik_foot_root, loc=<14.8262 -7.3545 8.296>, rot=<0.0357 0.7045 0.0284 -0.7082>}
 ik_foot_r {parent=ik_foot_root, loc=<-14.8262 -7.3546 8.2958>, rot=<0.7082 0.0284 -0.7045 0.0357>}
 ik_hand_root {parent=root, loc=<0 0 0>, rot=<0 0 0 1>}
 ik_hand_gun {parent=ik_hand_root, loc=<-48.1744 3.3546 104.4381>, rot=<0.6276 0.1309 0.4183 -0.6434>}
 ik_hand_l {parent=ik_hand_gun, loc=<59.3272 67.6997 34.3524>, rot=<-0.6158 -0.7027 -0.3565 0>}
 ik_hand_r {parent=ik_hand_gun, loc=<0 0 0>, rot=<0 0 0 1>}
]
