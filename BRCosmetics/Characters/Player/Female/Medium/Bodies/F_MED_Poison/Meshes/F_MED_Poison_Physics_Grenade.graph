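# Physics asset for the grenade on the F_MED_Poison female character body, apparently for secondary dangle motion, given the dyn_grenade bone name.
# It holds two capsule bodies: a small one on dyn_grenade and a longer, angled one on pelvis.
# A single constraint links the grenade to the pelvis with tight swing and twist limit angles.
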PhysicsAsset /BRCosmetics/Characters/Player/Female/Medium/Bodies/F_MED_Poison/Meshes/F_MED_Poison_Physics_Grenade
bodies {
 dyn_grenade [{shapeType=Capsule, length=7.5084, radius=4.2067, shapeName=dyn_grenade_capsule, center=<4 3 0>, rotation=<rot 0 -90 0>}]
 pelvis [{shapeType=Capsule, length=41.375, radius=6, center=<-12.7926 0.5107 -7.1509>, rotation=<rot 81.2249 138.9895 138.5237>}]
}
constraints [{bone1Name=dyn_grenade, bone2Name=pelvis, swing1LimitDegrees=15, swing2LimitDegrees=20, twistLimitDegrees=5}]
