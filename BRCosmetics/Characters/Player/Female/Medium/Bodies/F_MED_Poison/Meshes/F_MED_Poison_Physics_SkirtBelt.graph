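# Physics asset for the F_MED_Poison female character's secondary motion on skirt, belts, mask and wrist straps.
# Capsules on pelvis, spine and thighs accompany limited-swing chains: three skirt strips (back pair, front-left) and front and back belts from the pelvis.
# Front and back belt ends are cross-locked into a loop, with a free constraint to the left thigh.
# A dangling belt branches off the front, a mask chain off the back.
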PhysicsAsset /BRCosmetics/Characters/Player/Female/Medium/Bodies/F_MED_Poison/Meshes/F_MED_Poison_Physics_SkirtBelt
bodies {
 dyn_belt_00 [{shapeType=Capsule, length=2.5, radius=1, center=<1.6512 0.2017 0.7876>, rotation=<rot 0 0 90>}]
 dyn_belt_01 [{shapeType=Capsule, length=2.5, radius=1, center=<1.8392 0.0285 0>, rotation=<rot 0 0.8887 -139.8907>}]
 dyn_belt_02 [{shapeType=Capsule, length=4, radius=7, center=<3 0 5>, rotation=<rot 90 0 0>}]
 dyn_belt_03 [{shapeType=Capsule, length=4, radius=7, center=<3 0 4>, rotation=<rot 90 0 0>}]
 dyn_belt_04 [{shapeType=Capsule, length=4, radius=7, center=<3 0 3.4698>, rotation=<rot 90 -179.9604 -179.9604>}]
 dyn_belt_05 [{shapeType=Capsule, length=4, radius=5, center=<3 0 3>, rotation=<rot 90 0 0>}]
 dyn_belt_06 [{shapeType=Capsule, length=4, radius=5, center=<3 0 3>, rotation=<rot 90 0 0>}]
 dyn_belt_bk_01 [{shapeType=Capsule, length=4.8443, radius=2.3981, center=<2.9676 -0.2614 0.5272>, rotation=<rot 90 0 0>}]
 dyn_belt_bk_02 [{shapeType=Capsule, length=3.4281, radius=3.8143, center=<2.9676 -0.2614 0.5272>, rotation=<rot 90 0 0>}]
 dyn_belt_bk_03 [{shapeType=Capsule, length=4.2355, radius=3.0069, center=<2.9676 -0.2614 0.5272>, rotation=<rot 90 0 0>}]
 dyn_belt_bk_04 [{shapeType=Capsule, length=4.8443, radius=2.3981, center=<2.9676 -0.2614 0.5272>, rotation=<rot 90 0 0>}]
 dyn_belt_bk_05 [{shapeType=Capsule, length=4.8443, radius=2.3981, center=<2.9676 -0.2614 0.5272>, rotation=<rot 90 0 0>}]
 dyn_belt_bk_06 [{shapeType=Capsule, length=3, radius=5, center=<0 0 4>, rotation=<rot 90 0 0>}]
 dyn_belt_ft_00 [{shapeType=Capsule, length=4.8443, radius=2.3981, center=<2.9676 -0.2614 0.5272>, rotation=<rot 90 0 0>}]
 dyn_belt_ft_01 [{shapeType=Capsule, length=3, radius=2, center=<3 0 4>, rotation=<rot 90 0 0>}]
 dyn_belt_ft_02 [{shapeType=Capsule, length=3, radius=2, center=<3 0 4>, rotation=<rot 90 0 0>}]
 dyn_belt_ft_03 [{shapeType=Capsule, length=3, radius=2, center=<3 0 4>, rotation=<rot 90 0 0>}]
 dyn_belt_ft_04 [{shapeType=Capsule, length=3, radius=2, center=<3 0 4>, rotation=<rot 90 0 0>}]
 dyn_belt_ft_05 [{shapeType=Capsule, length=3, radius=6, center=<3 0 4>, rotation=<rot 90 0 0>}]
 dyn_belt_ft_06 [{shapeType=Capsule, length=3, radius=2, center=<3 0 4>, rotation=<rot 90 0 0>}]
 dyn_belt_ft_base_00 [{shapeType=Capsule, length=4.8443, radius=2.3981, center=<2.9676 -0.2614 0.5272>, rotation=<rot 90 0 0>}]
 dyn_belt_ft_base_01 [{shapeType=Capsule, length=4.8443, radius=2.3981, center=<2.9676 -0.2614 0.5272>, rotation=<rot 90 0 0>}]
 dyn_mask_02 [{shapeType=Capsule, length=0.505, radius=0.505, rotation=<rot 0 -90 0>}]
 dyn_mask_04 [{shapeType=Capsule, length=1, radius=1, shapeName=dyn_mask_04_capsule, center=<1 0 0>, rotation=<rot 90 0 0>}]
 dyn_mask_05 [{shapeType=Capsule, length=1, radius=1, shapeName=dyn_mask_05_capsule, center=<1 0 0>, rotation=<rot 90 0 0>}]
 dyn_mask_06 [{shapeType=Capsule, length=5, radius=5, shapeName=dyn_mask_06_capsule, center=<4 0 0>, rotation=<rot 90 0 0>}]
 dyn_skirt_bk_01_l [{shapeType=Capsule, length=5, radius=6, center=<6 0 4>, rotation=<rot 0 0 90>}]
 dyn_skirt_bk_01_r [{shapeType=Capsule, length=5, radius=6, center=<6 0 4>, rotation=<rot 0 0 90>}]
 dyn_skirt_bk_02_l [{shapeType=Capsule, length=5, radius=6, center=<6 0 4>, rotation=<rot 0 0 90>}]
 dyn_skirt_bk_02_r [{shapeType=Capsule, length=5, radius=6, center=<6 0 4>, rotation=<rot 0 0 90>}]
 dyn_skirt_bk_03_l [{shapeType=Capsule, length=5, radius=6, center=<6 0 4>, rotation=<rot 0 0 90>}]
 dyn_skirt_bk_03_r [{shapeType=Capsule, length=5, radius=6, center=<6 0 4>, rotation=<rot 0 0 90>}]
 dyn_skirt_fr_01_l [{shapeType=Capsule, length=5, radius=6, center=<6 0 4>, rotation=<rot 0 0 90>}]
 dyn_skirt_fr_02_l [{shapeType=Capsule, length=1, radius=6, center=<6 0 4>, rotation=<rot 0 0 90>}]
 dyn_skirt_fr_03_l [{shapeType=Capsule, length=2.75, radius=5.75, center=<4 0 4>, rotation=<rot 0 0 90>}]
 dyn_wrist_strap_l [{shapeType=Box, extentX=3.51, extentY=1.01, extentZ=4.76, center=<1.8457 -0.021 2.5651>, rotation=<rot 50.7789 3.9509 147.0448>}]
 dyn_wrist_strap_l1 [{shapeType=Box, extentX=3.51, extentY=1.01, extentZ=4.76, center=<-3.5686 -0.2213 0.0493>, rotation=<rot 2.9755 -9.6063 89.9932>}]
 dyn_wrist_strap_l_MSTR [{shapeType=Capsule, length=0.505, radius=0.505, rotation=<rot 0 -90 0>}]
 dyn_wrist_strap_l_MSTR1 [{shapeType=Capsule, length=0.505, radius=0.505, rotation=<rot 0 -90 0>}]
 pelvis [{shapeType=Capsule, length=41.375, radius=6, center=<-12.7926 0.5107 -7.1509>, rotation=<rot 81.2249 138.9895 138.5237>}]
 spine_02 [{shapeType=Capsule, length=0.5864, radius=8.1084, center=<2.3392 2.274 0.2246>, rotation=<rot -1.2787 -90 0.7891>}]
 thigh_twist_01_l [{shapeType=Capsule, length=23.9629, radius=7.7583, center=<11.4781 -0.8688 0.7247>, rotation=<rot 0.1089 -91.2953 -90.6794>}]
 thigh_twist_01_r [{shapeType=Capsule, length=33.3253, radius=6.7893, center=<-5.0069 0.3425 0>, rotation=<rot 0 89.9999 89.8962>}]
}
constraints [
 {
  bone1Name=thigh_twist_01_l
  bone2Name=pelvis
  swing1LimitDegrees=45
  swing1Motion=Free
  swing2LimitDegrees=45
  swing2Motion=Free
  twistLimitDegrees=45
  twistMotion=Free
 }
 {
  bone1Name=spine_02
  bone2Name=pelvis
  swing1LimitDegrees=45
  swing1Motion=Free
  swing2LimitDegrees=45
  swing2Motion=Free
  twistLimitDegrees=45
  twistMotion=Free
 }
 {
  bone1Name=thigh_twist_01_r
  bone2Name=pelvis
  swing1LimitDegrees=45
  swing1Motion=Free
  swing2LimitDegrees=45
  swing2Motion=Free
  twistLimitDegrees=45
  twistMotion=Free
 }
 {bone1Name=dyn_skirt_bk_03_r, bone2Name=dyn_skirt_bk_02_r, swing1LimitDegrees=30, swing2LimitDegrees=30, twistLimitDegrees=1}
 {bone1Name=dyn_skirt_bk_02_r, bone2Name=dyn_skirt_bk_01_r, swing1LimitDegrees=30, swing2LimitDegrees=30, twistLimitDegrees=1}
 {bone1Name=dyn_skirt_bk_01_r, bone2Name=pelvis, swing1LimitDegrees=30, swing2LimitDegrees=30, twistLimitDegrees=1}
 {bone1Name=dyn_skirt_bk_03_l, bone2Name=dyn_skirt_bk_02_l, swing1LimitDegrees=30, swing2LimitDegrees=30, twistLimitDegrees=1}
 {bone1Name=dyn_skirt_bk_02_l, bone2Name=dyn_skirt_bk_01_l, swing1LimitDegrees=30, swing2LimitDegrees=30, twistLimitDegrees=1}
 {bone1Name=dyn_skirt_fr_03_l, bone2Name=dyn_skirt_fr_02_l, swing1LimitDegrees=30, swing2LimitDegrees=30, twistLimitDegrees=1}
 {bone1Name=dyn_skirt_bk_01_l, bone2Name=pelvis, swing1LimitDegrees=30, swing2LimitDegrees=30, twistLimitDegrees=1}
 {bone1Name=dyn_skirt_fr_02_l, bone2Name=dyn_skirt_fr_01_l, swing1LimitDegrees=30, swing2LimitDegrees=30, twistLimitDegrees=1}
 {bone1Name=dyn_skirt_fr_01_l, bone2Name=pelvis, swing1LimitDegrees=30, swing2LimitDegrees=30, twistLimitDegrees=1}
 {
  bone1Name=dyn_mask_02
  bone2Name=pelvis
  swing1LimitDegrees=45
  swing1Motion=Free
  swing2LimitDegrees=45
  swing2Motion=Free
  twistLimitDegrees=45
  twistMotion=Free
 }
 {bone1Name=dyn_wrist_strap_l, bone2Name=dyn_wrist_strap_l_MSTR, swing1LimitDegrees=45, swing2LimitDegrees=45, twistLimitDegrees=1}
 {bone1Name=dyn_wrist_strap_l1, bone2Name=dyn_wrist_strap_l_MSTR1, swing1LimitDegrees=45, swing2LimitDegrees=45, twistLimitDegrees=1}
 {
  bone1Name=dyn_wrist_strap_l_MSTR
  bone2Name=spine_02
  swing1LimitDegrees=45
  swing1Motion=Free
  swing2LimitDegrees=45
  swing2Motion=Free
  twistLimitDegrees=45
  twistMotion=Free
 }
 {
  bone1Name=dyn_wrist_strap_l_MSTR1
  bone2Name=spine_02
  swing1LimitDegrees=45
  swing1Motion=Free
  swing2LimitDegrees=45
  swing2Motion=Free
  twistLimitDegrees=45
  twistMotion=Free
 }
 {bone1Name=dyn_belt_ft_02, bone2Name=dyn_belt_ft_01, swing1LimitDegrees=20, swing2LimitDegrees=20, twistLimitDegrees=20}
 {bone1Name=dyn_belt_ft_03, bone2Name=dyn_belt_ft_02, swing1LimitDegrees=20, swing2LimitDegrees=20, twistLimitDegrees=20}
 {
  bone1Name=dyn_belt_ft_04
  bone2Name=dyn_belt_ft_03
  swing1LimitDegrees=25
  swing1Motion=Locked
  swing2LimitDegrees=5
  swing2Motion=Locked
  twistLimitDegrees=1
  twistMotion=Locked
 }
 {
  bone1Name=dyn_belt_ft_05
  bone2Name=dyn_belt_ft_04
  swing1LimitDegrees=1
  swing1Motion=Locked
  swing2LimitDegrees=1
  swing2Motion=Locked
  twistLimitDegrees=5
  twistMotion=Locked
 }
 {
  bone1Name=dyn_belt_ft_06
  bone2Name=dyn_belt_ft_05
  swing1LimitDegrees=1
  swing1Motion=Locked
  swing2LimitDegrees=1
  swing2Motion=Locked
  twistLimitDegrees=1
  twistMotion=Locked
 }
 {
  bone1Name=thigh_twist_01_l
  bone2Name=dyn_belt_ft_06
  swing1LimitDegrees=45
  swing1Motion=Free
  swing2LimitDegrees=45
  swing2Motion=Free
  twistLimitDegrees=45
  twistMotion=Free
 }
 {bone1Name=dyn_belt_ft_base_01, bone2Name=dyn_belt_ft_base_00, swing1LimitDegrees=20, swing2LimitDegrees=5, twistLimitDegrees=20}
 {bone1Name=dyn_belt_ft_base_00, bone2Name=pelvis, swing1LimitDegrees=20, swing2LimitDegrees=5, twistLimitDegrees=20}
 {bone1Name=dyn_belt_ft_00, bone2Name=dyn_belt_ft_base_01, swing1LimitDegrees=20, swing2LimitDegrees=10, twistLimitDegrees=20}
 {bone1Name=dyn_belt_ft_01, bone2Name=dyn_belt_ft_00, swing1LimitDegrees=20, swing2LimitDegrees=20, twistLimitDegrees=20}
 {bone1Name=dyn_belt_00, bone2Name=dyn_belt_ft_00, swing1LimitDegrees=5, swing2LimitDegrees=5, twistLimitDegrees=1}
 {bone1Name=dyn_belt_01, bone2Name=dyn_belt_00, swing1LimitDegrees=45, swing2LimitDegrees=45, twistLimitDegrees=1}
 {bone1Name=dyn_belt_02, bone2Name=dyn_belt_01, swing1LimitDegrees=45, swing2LimitDegrees=45, twistLimitDegrees=1}
 {bone1Name=dyn_belt_03, bone2Name=dyn_belt_02, swing1LimitDegrees=45, swing2LimitDegrees=45, twistLimitDegrees=1}
 {bone1Name=dyn_belt_04, bone2Name=dyn_belt_03, swing1LimitDegrees=45, swing2LimitDegrees=45, twistLimitDegrees=1}
 {bone1Name=dyn_belt_05, bone2Name=dyn_belt_04, swing1LimitDegrees=45, swing2LimitDegrees=45, twistLimitDegrees=1}
 {bone1Name=dyn_belt_06, bone2Name=dyn_belt_05, swing1LimitDegrees=45, swing2LimitDegrees=45, twistLimitDegrees=1}
 {bone1Name=dyn_belt_bk_01, bone2Name=pelvis, swing1LimitDegrees=20, swing2LimitDegrees=20, twistLimitDegrees=20}
 {bone1Name=dyn_belt_bk_02, bone2Name=dyn_belt_bk_01, swing1LimitDegrees=20, swing2LimitDegrees=20, twistLimitDegrees=5}
 {bone1Name=dyn_belt_bk_03, bone2Name=dyn_belt_bk_02, swing1LimitDegrees=20, swing2LimitDegrees=20, twistLimitDegrees=5}
 {
  bone1Name=dyn_belt_bk_04
  bone2Name=dyn_belt_bk_03
  swing1LimitDegrees=20
  swing1Motion=Locked
  swing2LimitDegrees=20
  twistLimitDegrees=20
  twistMotion=Locked
 }
 {
  bone1Name=dyn_belt_bk_05
  bone2Name=dyn_belt_bk_04
  swing1LimitDegrees=45
  swing1Motion=Locked
  swing2LimitDegrees=45
  swing2Motion=Locked
  twistLimitDegrees=45
  twistMotion=Locked
 }
 {
  bone1Name=dyn_belt_bk_06
  bone2Name=dyn_belt_bk_05
  swing1LimitDegrees=45
  swing1Motion=Locked
  swing2LimitDegrees=45
  swing2Motion=Locked
  twistLimitDegrees=45
  twistMotion=Locked
 }
 {
  bone1Name=dyn_belt_bk_05
  bone2Name=dyn_belt_ft_06
  swing1LimitDegrees=1
  swing1Motion=Locked
  swing2LimitDegrees=1
  swing2Motion=Locked
  twistLimitDegrees=1
  twistMotion=Locked
 }
 {
  bone1Name=dyn_belt_bk_06
  bone2Name=dyn_belt_ft_05
  swing1LimitDegrees=1
  swing1Motion=Locked
  swing2LimitDegrees=1
  swing2Motion=Locked
  twistLimitDegrees=1
  twistMotion=Locked
 }
 {bone1Name=dyn_mask_04, bone2Name=dyn_belt_bk_04, swing1LimitDegrees=2, swing2LimitDegrees=20, twistLimitDegrees=2}
 {bone1Name=dyn_mask_05, bone2Name=dyn_mask_04, swing1LimitDegrees=3, swing2LimitDegrees=30, twistLimitDegrees=3}
 {bone1Name=dyn_mask_06, bone2Name=dyn_mask_05, swing1LimitDegrees=3, swing2LimitDegrees=45, twistLimitDegrees=3}
]
