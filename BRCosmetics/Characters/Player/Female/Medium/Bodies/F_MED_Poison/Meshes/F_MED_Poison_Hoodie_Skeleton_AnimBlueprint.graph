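# Post-animation secondary-motion AnimBlueprint for the F_MED_Poison hoodie body cosmetic.
# Copies the attached parent's pose, then chains bone constraints and additive corrective offsets for collar, cuffs, lapels and belt, weighted by head, arm and thigh variables.
# AnimDynamics springs jiggle jacket, cuffs, skirt and pants; a Control Rig pass precedes RigidBody simulations for skirt-belt and grenade.
# Lapel correctives use a fixed low weight during emotes or in the front end.
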AnimBlueprint /BRCosmetics/Characters/Player/Female/Medium/Bodies/F_MED_Poison/Meshes/F_MED_Poison_Hoodie_Skeleton_AnimBlueprint
target_skeleton=&/F_MED_Poison_Hoodie_Skeleton
graphs [
 AnimGraph {nodes [
   Anim Dynamics - Bone: dyn_skirt_bk_03_l!(ComponentPose, ExternalForce="0,0,0") -> (Pose) {
    alpha=0.6
    angularDampingOverride=3
    angularSpringConstant=50
    bAngularSpring=true
    bGravityOverrideInSimSpace=true
    bUseGravityOverride=true
    simulationSpace=BoneRelative
    alphaBoolBlend {blendOption=Linear}
    boundBone {boneName=dyn_skirt_bk_03_l}
    componentAppliedLinearAccClamp=<5000 5000 5000>
    componentLinearAccScale=<0.2 0.2 0.55>
    componentLinearVelScale=<0.2 0.2 0.3>
    gravityOverride {z=-980}
    physicsBodyDefinitions [{
      sphereCollisionRadius=0
      boundBone {boneName=dyn_skirt_bk_03_l}
      boxExtents=<5 25 25>
      constraintSetup {angularLimitsMax=<0 20 0>, angularLimitsMin=<-0.1 -80 -0.1>, angularTarget=<1 0 0>}
      localJointOffset=<8 0 0>
     }]
    relativeSpaceBone {boneName=root}
   }
   Anim Dynamics - Bone: dyn_skirt_bk_03_r!(ComponentPose, ExternalForce="0,0,0") -> (Pose) {
    alpha=0.6
    angularDampingOverride=3
    angularSpringConstant=50
    bAngularSpring=true
    bGravityOverrideInSimSpace=true
    bUseGravityOverride=true
    simulationSpace=BoneRelative
    alphaBoolBlend {blendOption=Linear}
    boundBone {boneName=dyn_skirt_bk_03_r}
    componentAppliedLinearAccClamp=<5000 5000 5000>
    componentLinearAccScale=<0.2 0.2 0.55>
    componentLinearVelScale=<0.2 0.2 0.3>
    gravityOverride {z=-980}
    physicsBodyDefinitions [{
      sphereCollisionRadius=0
      boundBone {boneName=dyn_skirt_bk_03_r}
      boxExtents=<5 25 25>
      constraintSetup {angularLimitsMax=<0 20 0>, angularLimitsMin=<-0.1 -80 -0.1>, angularTarget=<1 0 0>}
      localJointOffset=<8 0 0>
     }]
    relativeSpaceBone {boneName=root}
   }
   Anim Dynamics - Bone: dyn_grenade(Alpha=1, ExternalForce="0,0,0") {
    angularDampingOverride=0.3
    angularSpringConstant=100
    bAngularSpring=true
    bGravityOverrideInSimSpace=true
    bUseGravityOverride=true
    simulationSpace=BoneRelative
    alphaBoolBlend {blendOption=Linear}
    boundBone {boneName=dyn_grenade}
    componentLinearAccScale=<0.8 0.8 0.8>
    componentLinearVelScale=<0.4 0.4 0.4>
    gravityOverride {z=-980}
    physicsBodyDefinitions [{
      sphereCollisionRadius=0
      boundBone {boneName=dyn_grenade}
      boxExtents=<5 10 10>
      constraintSetup {angularLimitsMax=<10 10 10>, angularLimitsMin=<-14 -10 -10>, angularTarget=<1 0 0>}
      localJointOffset=<1 0 0>
     }]
    relativeSpaceBone {boneName=root}
   }
   n1 = Anim Dynamics - Bone: dyn_skirt_bk_03_l(ExternalForce="0,0,0") {
    alpha=0.6
    angularDampingOverride=3
    angularSpringConstant=50
    bAngularSpring=true
    bGravityOverrideInSimSpace=true
    bUseGravityOverride=true
    simulationSpace=BoneRelative
    alphaBoolBlend {blendOption=Linear}
    boundBone {boneName=dyn_skirt_bk_03_l}
    componentAppliedLinearAccClamp=<5000 5000 5000>
    componentLinearAccScale=<0.2 0.2 0.55>
    componentLinearVelScale=<0.2 0.2 0.3>
    gravityOverride {z=-980}
    physicsBodyDefinitions [{
      sphereCollisionRadius=0
      boundBone {boneName=dyn_skirt_bk_03_l}
      boxExtents=<5 25 25>
      constraintSetup {angularLimitsMax=<0 40 0>, angularLimitsMin=<-0.1 -80 -0.1>, angularTarget=<1 0 0>}
      localJointOffset=<6 0 0>
     }]
    relativeSpaceBone {boneName=root}
   }
   n2 = Anim Dynamics - Bone: dyn_skirt_bk_03_r(ComponentPose=@n1, ExternalForce="0,0,0") {
    alpha=0.6
    angularDampingOverride=3
    angularSpringConstant=50
    bAngularSpring=true
    bGravityOverrideInSimSpace=true
    bUseGravityOverride=true
    simulationSpace=BoneRelative
    alphaBoolBlend {blendOption=Linear}
    boundBone {boneName=dyn_skirt_bk_03_r}
    componentAppliedLinearAccClamp=<5000 5000 5000>
    componentLinearAccScale=<0.2 0.2 0.55>
    componentLinearVelScale=<0.2 0.2 0.3>
    gravityOverride {z=-980}
    physicsBodyDefinitions [{
      sphereCollisionRadius=0
      boundBone {boneName=dyn_skirt_bk_03_r}
      boxExtents=<5 25 25>
      constraintSetup {angularLimitsMax=<0 40 0>, angularLimitsMin=<-0.1 -80 -0.1>, angularTarget=<1 0 0>}
      localJointOffset=<6 0 0>
     }]
    relativeSpaceBone {boneName=root}
   }
   n3 = Anim Dynamics - Bone: dyn_skirt_side_03_r!(ComponentPose=@n2, ExternalForce="0,0,0") -> (Pose) {
    alpha=0.6
    angularDampingOverride=3
    angularSpringConstant=50
    bAngularSpring=true
    bGravityOverrideInSimSpace=true
    bUseGravityOverride=true
    simulationSpace=BoneRelative
    alphaBoolBlend {blendOption=Linear}
    boundBone {boneName=dyn_skirt_side_03_r}
    componentAppliedLinearAccClamp=<5000 5000 5000>
    componentLinearAccScale=<0.2 0.2 0.55>
    componentLinearVelScale=<0.2 0.2 0.3>
    gravityOverride {z=-980}
    physicsBodyDefinitions [{
      sphereCollisionRadius=0
      boundBone {boneName=dyn_skirt_side_03_r}
      boxExtents=<5 25 25>
      constraintSetup {angularLimitsMax=<0 20 0>, angularLimitsMin=<-0.1 -80 -0.1>, angularTarget=<1 0 0>}
      localJointOffset=<6 0 0>
     }]
    relativeSpaceBone {boneName=root}
   }
   Anim Dynamics - Bone: dyn_skirt_fr_03_l!(ComponentPose=@n3, ExternalForce="0,0,0") -> (Pose) {
    alpha=0.6
    angularDampingOverride=3
    angularSpringConstant=50
    bAngularSpring=true
    bGravityOverrideInSimSpace=true
    bUseGravityOverride=true
    simulationSpace=BoneRelative
    alphaBoolBlend {blendOption=Linear}
    boundBone {boneName=dyn_skirt_fr_03_l}
    componentAppliedLinearAccClamp=<5000 5000 5000>
    componentLinearAccScale=<0.2 0.2 0.55>
    componentLinearVelScale=<0.2 0.2 0.3>
    gravityOverride {z=-980}
    physicsBodyDefinitions [{
      sphereCollisionRadius=0
      boundBone {boneName=dyn_skirt_fr_03_l}
      boxExtents=<5 25 25>
      constraintSetup {angularLimitsMax=<0 40 0>, angularLimitsMin=<-0.1 -80 -0.1>, angularTarget=<1 0 0>}
      localJointOffset=<6 0 0>
     }]
    relativeSpaceBone {boneName=root}
   }
   Component To Local()
   "Transform (Modify) Bone - Bone: dyn_skirt_fr_mid_r_MSTR"(Alpha=Get thigh_rot_r(), Translation="-10,-10,10", Rotation="0,0,0", Scale="0,0,0") {
    rotationMode=BMM_Additive
    rotationSpace=BCS_BoneSpace
    scaleSpace=BCS_BoneSpace
    translationMode=BMM_Additive
    translationSpace=BCS_BoneSpace
    alphaBoolBlend {blendOption=Linear}
    boneToModify {boneName=dyn_skirt_fr_mid_r_MSTR}
    scale=<0 0 0>
    translation=<-10 -10 10>
   }
   n4 = RigidBody(ExternalForce=Get finalGravity() -> (finalGravity:Vector)) {
    bClampLinearTranslationLimitToRefPose=true
    bOverrideWorldGravity=true
    overridePhysicsAsset=&/F_MED_Poison_Physics_SkirtBelt
    simulationSpace=BaseBoneSpace
    alphaBoolBlend {blendOption=Linear}
    baseBoneRef {boneName=spine_01}
    componentLinearAccScale=<0.4 0.4 0.6>
    componentLinearVelScale=<0.1 0.1 0.3>
    overrideWorldGravity {z=-980}
   }
   RigidBody(ComponentPose=@n4) {
    bClampLinearTranslationLimitToRefPose=true
    bOverrideWorldGravity=true
    overridePhysicsAsset=&/F_MED_Poison_Physics_Grenade
    simulationSpace=BaseBoneSpace
    alphaBoolBlend {blendOption=Linear}
    baseBoneRef {boneName=spine_01}
    componentLinearAccScale=<0.4 0.4 0.6>
    componentLinearVelScale=<0.1 0.1 0.3>
    overrideWorldGravity {z=-980}
   }
   n5 = Get JacketExternalForce() -> (JacketExternalForce:Vector)
   n6 = Get Head Rot Up() -> (Head Rot Up:double)
   n7 = "Transform (Modify) Bone - Bone: dyn_collar_01"(
    ComponentPose=Local To Component(LocalPose=Copy Pose From Mesh() {bUseAttachedParent=true})
    Alpha=@n6
    Translation="0,0,-3"
    Rotation="0,0,0"
    Scale="0,0,0"
   ) {
    rotationMode=BMM_Additive
    rotationSpace=BCS_BoneSpace
    scaleSpace=BCS_BoneSpace
    translationMode=BMM_Additive
    translationSpace=BCS_BoneSpace
    alphaBoolBlend {blendOption=Linear}
    boneToModify {boneName=dyn_collar_01}
    scale=<0 0 0>
    translation {z=-3}
   }
   n8 = "Transform (Modify) Bone - Bone: dyn_collar_00"(ComponentPose=@n7, Alpha=@n6, Translation="0,0,-1", Rotation="0,0,0", Scale="0,0,0") {
    rotationMode=BMM_Additive
    rotationSpace=BCS_BoneSpace
    scaleSpace=BCS_BoneSpace
    translationMode=BMM_Additive
    translationSpace=BCS_BoneSpace
    alphaBoolBlend {blendOption=Linear}
    boneToModify {boneName=dyn_collar_00}
    scale=<0 0 0>
    translation {z=-1}
   }
   n9 = "Constraint - dyn_clip_front_MSTR (PPPPT)"!(
    ComponentPose=@n8
    Alpha=1
    ConstraintWeights_0=0.0723
    ConstraintWeights_1:float=0.1633
    ConstraintWeights_2:float=0.2784
    ConstraintWeights_3:float=0.4859
    ConstraintWeights_4:float=0
   ) -> (Pose) {
    alphaBoolBlend {blendOption=Linear}
    boneToModify {boneName=dyn_clip_front_MSTR}
    constraintSetup [
     {targetBone {boneName=spine_01}}
     {targetBone {boneName=spine_02}}
     {targetBone {boneName=spine_03}}
     {targetBone {boneName=spine_04}}
     {transformType=Translation, targetBone {boneName=None}}
    ]
    constraintWeights [0,1,1,1,1]
   }
   n10 = Anim Dynamics - Bone: dyn_jacketFront_r(ComponentPose=@n9, Alpha=1, ExternalForce=@n5) {
    angularDampingOverride=1
    angularSpringConstant=4000
    bAngularSpring=true
    bGravityOverrideInSimSpace=true
    bOverrideAngularDamping=true
    bUseGravityOverride=true
    simulationSpace=BoneRelative
    alphaBoolBlend {blendOption=Linear}
    boundBone {boneName=dyn_jacketFront_r}
    componentAppliedLinearAccClamp=<3000 3000 3000>
    componentLinearAccScale=<0.4 0.4 0.4>
    componentLinearVelScale=<0.2 0.2 0.2>
    gravityOverride {z=-980}
    physicsBodyDefinitions [{
      sphereCollisionRadius=0
      boundBone {boneName=dyn_jacketFront_r}
      boxExtents=<15 3 15>
      constraintSetup {angularLimitsMax=<2 5 2>, angularLimitsMin=<-2 -5 -60>, angularTarget=<1 0 0>}
      localJointOffset=<10 0 0>
     }]
    relativeSpaceBone {boneName=root}
   }
   n11 = Anim Dynamics - Bone: dyn_jacketFront_l(ComponentPose=@n10, Alpha=1, ExternalForce=@n5) {
    angularDampingOverride=1
    angularSpringConstant=4000
    bAngularSpring=true
    bGravityOverrideInSimSpace=true
    bOverrideAngularDamping=true
    bUseGravityOverride=true
    simulationSpace=BoneRelative
    alphaBoolBlend {blendOption=Linear}
    boundBone {boneName=dyn_jacketFront_l}
    componentAppliedLinearAccClamp=<3000 3000 3000>
    componentLinearAccScale=<0.4 0.4 0.4>
    componentLinearVelScale=<0.2 0.2 0.2>
    gravityOverride {z=-980}
    physicsBodyDefinitions [{
      sphereCollisionRadius=0
      boundBone {boneName=dyn_jacketFront_l}
      boxExtents=<15 3 15>
      constraintSetup {angularLimitsMax=<2 5 2>, angularLimitsMin=<-2 -5 -60>, angularTarget=<1 0 0>}
      localJointOffset=<-10 0 0>
     }]
    relativeSpaceBone {boneName=root}
   }
   n12 = Anim Dynamics - Bone: dyn_jacketSide_r(ComponentPose=@n11, Alpha=1, ExternalForce=@n5) {
    angularDampingOverride=1
    angularSpringConstant=4000
    bAngularSpring=true
    bGravityOverrideInSimSpace=true
    bOverrideAngularDamping=true
    bUseGravityOverride=true
    simulationSpace=BoneRelative
    alphaBoolBlend {blendOption=Linear}
    boundBone {boneName=dyn_jacketSide_r}
    componentAppliedLinearAccClamp=<3000 3000 3000>
    componentLinearAccScale=<0.4 0.4 0.4>
    componentLinearVelScale=<0.2 0.2 0.2>
    gravityOverride {z=-980}
    physicsBodyDefinitions [{
      sphereCollisionRadius=0
      boundBone {boneName=dyn_jacketSide_r}
      boxExtents=<15 3 15>
      constraintSetup {angularLimitsMax=<2 5 2>, angularLimitsMin=<-2 -5 -60>, angularTarget=<1 0 0>}
      localJointOffset=<10 0 0>
     }]
    relativeSpaceBone {boneName=root}
   }
   n13 = Anim Dynamics - Bone: dyn_jacketSide_l(ComponentPose=@n12, Alpha=1, ExternalForce=@n5) {
    angularDampingOverride=1
    angularSpringConstant=4000
    bAngularSpring=true
    bGravityOverrideInSimSpace=true
    bOverrideAngularDamping=true
    bUseGravityOverride=true
    simulationSpace=BoneRelative
    alphaBoolBlend {blendOption=Linear}
    boundBone {boneName=dyn_jacketSide_l}
    componentAppliedLinearAccClamp=<3000 3000 3000>
    componentLinearAccScale=<0.4 0.4 0.4>
    componentLinearVelScale=<0.2 0.2 0.2>
    gravityOverride {z=-980}
    physicsBodyDefinitions [{
      sphereCollisionRadius=0
      boundBone {boneName=dyn_jacketSide_l}
      boxExtents=<15 3 15>
      constraintSetup {angularLimitsMax=<2 5 2>, angularLimitsMin=<-2 -5 -60>, angularTarget=<1 0 0>}
      localJointOffset=<-10 0 0>
     }]
    relativeSpaceBone {boneName=root}
   }
   n14 = Anim Dynamics - Bone: dyn_jacketSideBack_r!(ComponentPose=@n13, bAlphaBoolEnabled:bool=True, ExternalForce=@n5) -> (Pose) {
    alphaInputType=Bool
    angularDampingOverride=1
    angularSpringConstant=4000
    bAngularSpring=true
    bGravityOverrideInSimSpace=true
    bOverrideAngularDamping=true
    bUseGravityOverride=true
    simulationSpace=BoneRelative
    alphaBoolBlend {blendOption=Linear}
    boundBone {boneName=dyn_jacketSideBack_r}
    componentAppliedLinearAccClamp=<3000 3000 3000>
    componentLinearAccScale=<0.4 0.4 0.4>
    componentLinearVelScale=<0.2 0.2 0.2>
    gravityOverride {z=-980}
    physicsBodyDefinitions [{
      sphereCollisionRadius=0
      boundBone {boneName=dyn_jacketSideBack_r}
      boxExtents=<15 3 15>
      constraintSetup {angularLimitsMax=<2 5 2>, angularLimitsMin=<-2 -5 -68>, angularTarget=<1 0 0>}
      localJointOffset=<10 0 0>
     }]
    relativeSpaceBone {boneName=root}
   }
   n15 = Anim Dynamics - Bone: dyn_jacketSideBack_l!(ComponentPose=@n14, bAlphaBoolEnabled:bool=True, ExternalForce=@n5) -> (Pose) {
    alphaInputType=Bool
    angularDampingOverride=1
    angularSpringConstant=4000
    bAngularSpring=true
    bGravityOverrideInSimSpace=true
    bOverrideAngularDamping=true
    bUseGravityOverride=true
    simulationSpace=BoneRelative
    alphaBoolBlend {blendOption=Linear}
    boundBone {boneName=dyn_jacketSideBack_l}
    componentAppliedLinearAccClamp=<3000 3000 3000>
    componentLinearAccScale=<0.4 0.4 0.4>
    componentLinearVelScale=<0.2 0.2 0.2>
    gravityOverride {z=-980}
    physicsBodyDefinitions [{
      sphereCollisionRadius=0
      boundBone {boneName=dyn_jacketSideBack_l}
      boxExtents=<15 3 15>
      constraintSetup {angularLimitsMax=<2 5 2>, angularLimitsMin=<-2 -5 -68>, angularTarget=<1 0 0>}
      localJointOffset=<-10 0 0>
     }]
    relativeSpaceBone {boneName=root}
   }
   n16 = Anim Dynamics - Bone: dyn_jacketBack(ComponentPose=@n15, Alpha=0.7, ExternalForce=@n5) {
    angularDampingOverride=1
    angularSpringConstant=4000
    bAngularSpring=true
    bGravityOverrideInSimSpace=true
    bOverrideAngularDamping=true
    bUseGravityOverride=true
    simulationSpace=BoneRelative
    alphaBoolBlend {blendOption=Linear}
    boundBone {boneName=dyn_jacketBack}
    componentAppliedLinearAccClamp=<3000 3000 3000>
    componentLinearAccScale=<0.4 0.4 0.4>
    componentLinearVelScale=<0.2 0.2 0.2>
    gravityOverride {z=-980}
    physicsBodyDefinitions [{
      sphereCollisionRadius=0
      boundBone {boneName=dyn_jacketBack}
      boxExtents=<15 3 15>
      constraintSetup {angularLimitsMax=<2 5 2>, angularLimitsMin=<-2 -5 -68>, angularTarget=<1 0 0>}
      localJointOffset=<10 0 0>
     }]
    relativeSpaceBone {boneName=spine_02}
   }
   n17 = "Constraint - dyn_belt_ft_03 (P)"(ComponentPose=@n16, Alpha=0, ConstraintWeights_0=1) {
    alphaBoolBlend {blendOption=Linear}
    boneToModify {boneName=dyn_belt_ft_03}
    constraintSetup [{targetBone {boneName=thigh_twist_01_l}}]
    constraintWeights [1]
   }
   n18 = "Constraint - dyn_belt_bk_03 (P)"(ComponentPose=@n17, Alpha=0, ConstraintWeights_0=1) {
    alphaBoolBlend {blendOption=Linear}
    boneToModify {boneName=dyn_belt_bk_03}
    constraintSetup [{targetBone {boneName=thigh_twist_01_l}}]
    constraintWeights [1]
   }
   n19 = Get r_armupClamp() -> (r_armupClamp:double)
   n20 = "Transform (Modify) Bone - Bone: dyn_cuff_r_02_MSTR"(ComponentPose=@n18, Alpha=@n19, Translation="0,5,-5", Rotation="0,0,0", Scale="0,0,0") {
    rotationMode=BMM_Additive
    rotationSpace=BCS_BoneSpace
    scaleSpace=BCS_BoneSpace
    translationMode=BMM_Additive
    translationSpace=BCS_BoneSpace
    alphaBoolBlend {blendOption=Linear}
    boneToModify {boneName=dyn_cuff_r_02_MSTR}
    scale=<0 0 0>
    translation {y=5, z=-5}
   }
   n21 = "Transform (Modify) Bone - Bone: dyn_cuff_r_04_MSTR"(ComponentPose=@n20, Alpha=@n19, Translation="0,-9,0", Rotation="0,0,0", Scale="0,0,0") {
    rotationMode=BMM_Additive
    rotationSpace=BCS_BoneSpace
    scaleSpace=BCS_BoneSpace
    translationMode=BMM_Additive
    translationSpace=BCS_BoneSpace
    alphaBoolBlend {blendOption=Linear}
    boneToModify {boneName=dyn_cuff_r_04_MSTR}
    scale=<0 0 0>
    translation {y=-9}
   }
   n22 = "Transform (Modify) Bone - Bone: dyn_cuff_r_05_MSTR"(ComponentPose=@n21, Alpha=@n19, Translation="4,0,0", Rotation="0,0,0", Scale="0,0,0") {
    rotationMode=BMM_Additive
    rotationSpace=BCS_BoneSpace
    scaleSpace=BCS_BoneSpace
    translationMode=BMM_Additive
    translationSpace=BCS_BoneSpace
    alphaBoolBlend {blendOption=Linear}
    boneToModify {boneName=dyn_cuff_r_05_MSTR}
    scale=<0 0 0>
    translation {x=4}
   }
   n23 = SelectFloat(
    A=0.15
    B=Get R_armupClamp2() -> (R_armupClamp2:double)
    bPickA=OR Boolean!(A=Get bIsPlayingEmote(), B=Get bIsInFrontEnd(), C:bool) -> (ReturnValue)
   )
   n24 = "Transform (Modify) Bone - Bone: dyn_lapel_r"(ComponentPose=@n22, Alpha=@n23, Translation="10,4,5", Rotation="-30,-110,0", Scale="0,0,0") {
    rotationMode=BMM_Additive
    rotationSpace=BCS_BoneSpace
    scaleSpace=BCS_BoneSpace
    translationMode=BMM_Additive
    translationSpace=BCS_BoneSpace
    alphaBoolBlend {blendOption=Linear}
    boneToModify {boneName=dyn_lapel_r}
    rotation {pitch=-30, yaw=-110}
    scale=<0 0 0>
    translation=<10 4 5>
   }
   n25 = "Transform (Modify) Bone - Bone: dyn_lapel_r1"(ComponentPose=@n24, Alpha=@n23, Translation="7,0,0", Rotation="0,-20,0", Scale="0,0,0") {
    rotationMode=BMM_Additive
    rotationSpace=BCS_BoneSpace
    scaleSpace=BCS_BoneSpace
    translationMode=BMM_Additive
    translationSpace=BCS_BoneSpace
    alphaBoolBlend {blendOption=Linear}
    boneToModify {boneName=dyn_lapel_r1}
    rotation {yaw=-20}
    scale=<0 0 0>
    translation {x=7}
   }
   n26 = "Transform (Modify) Bone - Bone: dyn_cuff_r_02_MSTR"(ComponentPose=@n25, Alpha=Get R_armupClamp2(), Translation="0,0,-10", Rotation="0,0,0", Scale="0,0,0") {
    rotationMode=BMM_Additive
    rotationSpace=BCS_BoneSpace
    scaleSpace=BCS_BoneSpace
    translationMode=BMM_Additive
    translationSpace=BCS_BoneSpace
    alphaBoolBlend {blendOption=Linear}
    boneToModify {boneName=dyn_cuff_r_02_MSTR}
    scale=<0 0 0>
    translation {z=-10}
   }
   n27 = Get bIsPlayingEmote()
   n28 = Get bIsInFrontEnd()
   n29 = "Transform (Modify) Bone - Bone: dyn_lapel_l"(
    ComponentPose=@n26
    Alpha=SelectFloat(A=0.15, B=Get r_armupClamp(), bPickA=OR Boolean!(A=@n27, B=@n28) -> (ReturnValue))
    Translation="0,0,0"
    Rotation="30,40,0"
    Scale="0,0,0"
   ) {
    rotationMode=BMM_Additive
    rotationSpace=BCS_BoneSpace
    scaleSpace=BCS_BoneSpace
    translationMode=BMM_Additive
    translationSpace=BCS_BoneSpace
    alphaBoolBlend {blendOption=Linear}
    boneToModify {boneName=dyn_lapel_l}
    rotation {pitch=30, yaw=40}
    scale=<0 0 0>
   }
   n30 = Anim Dynamics - Bone: dyn_cuff_l_01!(ComponentPose=@n29, ExternalForce="0,0,0") -> (Pose) {
    alpha=0.5
    angularDampingOverride=3
    angularSpringConstant=50
    bGravityOverrideInSimSpace=true
    bUseGravityOverride=true
    simulationSpace=BoneRelative
    alphaBoolBlend {blendOption=Linear}
    boundBone {boneName=dyn_cuff_l_01}
    componentAppliedLinearAccClamp=<5000 5000 5000>
    componentLinearAccScale=<0.2 0.2 0.55>
    componentLinearVelScale=<0.2 0.2 0.3>
    gravityOverride {z=-980}
    physicsBodyDefinitions [{
      sphereCollisionRadius=0
      boundBone {boneName=dyn_cuff_l_01}
      boxExtents=<10 30 30>
      constraintSetup {angularLimitsMax=<0 20 0>, angularLimitsMin=<-0.1 -80 -0.1>, angularTarget=<1 0 0>}
      localJointOffset=<6 0 0>
     }]
    relativeSpaceBone {boneName=root}
   }
   n31 = Anim Dynamics - Bone: dyn_cuff_r_01!(ComponentPose=@n30, ExternalForce="0,0,0") -> (Pose) {
    alpha=0.5
    angularDampingOverride=3
    angularSpringConstant=50
    bGravityOverrideInSimSpace=true
    bUseGravityOverride=true
    simulationSpace=BoneRelative
    alphaBoolBlend {blendOption=Linear}
    boundBone {boneName=dyn_cuff_r_01}
    componentAppliedLinearAccClamp=<5000 5000 5000>
    componentLinearAccScale=<0.2 0.2 0.55>
    componentLinearVelScale=<0.2 0.2 0.3>
    gravityOverride {z=-980}
    physicsBodyDefinitions [{
      sphereCollisionRadius=0
      boundBone {boneName=dyn_cuff_r_01}
      boxExtents=<10 30 30>
      constraintSetup {angularLimitsMax=<0 20 0>, angularLimitsMin=<-0.1 -80 -0.1>, angularTarget=<1 0 0>}
      localJointOffset=<6 0 0>
     }]
    relativeSpaceBone {boneName=root}
   }
   n32 = Anim Dynamics - Bone: dyn_cuff_l_04!(ComponentPose=@n31, ExternalForce="0,0,0") -> (Pose) {
    alpha=0.5
    angularDampingOverride=3
    angularSpringConstant=50
    bGravityOverrideInSimSpace=true
    bUseGravityOverride=true
    simulationSpace=BoneRelative
    alphaBoolBlend {blendOption=Linear}
    boundBone {boneName=dyn_cuff_l_04}
    componentAppliedLinearAccClamp=<5000 5000 5000>
    componentLinearAccScale=<0.2 0.2 0.55>
    componentLinearVelScale=<0.2 0.2 0.3>
    gravityOverride {z=-980}
    physicsBodyDefinitions [{
      sphereCollisionRadius=0
      boundBone {boneName=dyn_cuff_l_04}
      boxExtents=<10 30 30>
      constraintSetup {angularLimitsMax=<0 20 0>, angularLimitsMin=<-0.1 -80 -0.1>, angularTarget=<1 0 0>}
      localJointOffset=<6 0 0>
     }]
    relativeSpaceBone {boneName=root}
   }
   n33 = Anim Dynamics - Bone: dyn_cuff_r_04!(ComponentPose=@n32, ExternalForce="0,0,0") -> (Pose) {
    alpha=0.5
    angularDampingOverride=3
    angularSpringConstant=50
    bGravityOverrideInSimSpace=true
    bUseGravityOverride=true
    simulationSpace=BoneRelative
    alphaBoolBlend {blendOption=Linear}
    boundBone {boneName=dyn_cuff_r_04}
    componentAppliedLinearAccClamp=<5000 5000 5000>
    componentLinearAccScale=<0.2 0.2 0.55>
    componentLinearVelScale=<0.2 0.2 0.3>
    gravityOverride {z=-980}
    physicsBodyDefinitions [{
      sphereCollisionRadius=0
      boundBone {boneName=dyn_cuff_r_04}
      boxExtents=<10 30 30>
      constraintSetup {angularLimitsMax=<0 20 0>, angularLimitsMin=<-0.1 -80 -0.1>, angularTarget=<1 0 0>}
      localJointOffset=<6 0 0>
     }]
    relativeSpaceBone {boneName=root}
   }
   n34 = Anim Dynamics - Bone: dyn_cuff_l_02!(ComponentPose=@n33, ExternalForce="0,0,0") -> (Pose) {
    alpha=0.5
    angularDampingOverride=3
    angularSpringConstant=50
    bGravityOverrideInSimSpace=true
    bUseGravityOverride=true
    simulationSpace=BoneRelative
    alphaBoolBlend {blendOption=Linear}
    boundBone {boneName=dyn_cuff_l_02}
    componentAppliedLinearAccClamp=<5000 5000 5000>
    componentLinearAccScale=<0.2 0.2 0.55>
    componentLinearVelScale=<0.2 0.2 0.3>
    gravityOverride {z=-980}
    physicsBodyDefinitions [{
      sphereCollisionRadius=0
      boundBone {boneName=dyn_cuff_l_02}
      boxExtents=<10 30 30>
      constraintSetup {angularLimitsMax=<0 20 0>, angularLimitsMin=<-0.1 -80 -0.1>, angularTarget=<1 0 0>}
      localJointOffset=<6 0 0>
     }]
    relativeSpaceBone {boneName=root}
   }
   n35 = Anim Dynamics - Bone: dyn_cuff_r_02!(ComponentPose=@n34, ExternalForce="0,0,0") -> (Pose) {
    alpha=0.5
    angularDampingOverride=3
    angularSpringConstant=50
    bGravityOverrideInSimSpace=true
    bUseGravityOverride=true
    simulationSpace=BoneRelative
    alphaBoolBlend {blendOption=Linear}
    boundBone {boneName=dyn_cuff_r_02}
    componentAppliedLinearAccClamp=<5000 5000 5000>
    componentLinearAccScale=<0.2 0.2 0.55>
    componentLinearVelScale=<0.2 0.2 0.3>
    gravityOverride {z=-980}
    physicsBodyDefinitions [{
      sphereCollisionRadius=0
      boundBone {boneName=dyn_cuff_r_02}
      boxExtents=<10 30 30>
      constraintSetup {angularLimitsMax=<0 20 0>, angularLimitsMin=<-0.1 -80 -0.1>, angularTarget=<1 0 0>}
      localJointOffset=<6 0 0>
     }]
    relativeSpaceBone {boneName=root}
   }
   n36 = Anim Dynamics - Bone: dyn_cuff_l_03!(ComponentPose=@n35, ExternalForce="0,0,0") -> (Pose) {
    alpha=0.5
    angularDampingOverride=3
    angularSpringConstant=50
    bGravityOverrideInSimSpace=true
    bUseGravityOverride=true
    simulationSpace=BoneRelative
    alphaBoolBlend {blendOption=Linear}
    boundBone {boneName=dyn_cuff_l_03}
    componentAppliedLinearAccClamp=<5000 5000 5000>
    componentLinearAccScale=<0.2 0.2 0.55>
    componentLinearVelScale=<0.2 0.2 0.3>
    gravityOverride {z=-980}
    physicsBodyDefinitions [{
      sphereCollisionRadius=0
      boundBone {boneName=dyn_cuff_l_03}
      boxExtents=<10 30 30>
      constraintSetup {angularLimitsMax=<0 20 0>, angularLimitsMin=<-0.1 -80 -0.1>, angularTarget=<1 0 0>}
      localJointOffset=<6 0 0>
     }]
    relativeSpaceBone {boneName=root}
   }
   n37 = Anim Dynamics - Bone: dyn_cuff_r_03!(ComponentPose=@n36, ExternalForce="0,0,0") -> (Pose) {
    alpha=0.5
    angularDampingOverride=3
    angularSpringConstant=50
    bGravityOverrideInSimSpace=true
    bUseGravityOverride=true
    simulationSpace=BoneRelative
    alphaBoolBlend {blendOption=Linear}
    boundBone {boneName=dyn_cuff_r_03}
    componentAppliedLinearAccClamp=<5000 5000 5000>
    componentLinearAccScale=<0.2 0.2 0.55>
    componentLinearVelScale=<0.2 0.2 0.3>
    gravityOverride {z=-980}
    physicsBodyDefinitions [{
      sphereCollisionRadius=0
      boundBone {boneName=dyn_cuff_r_03}
      boxExtents=<10 30 30>
      constraintSetup {angularLimitsMax=<0 20 0>, angularLimitsMin=<-0.1 -80 -0.1>, angularTarget=<1 0 0>}
      localJointOffset=<6 0 0>
     }]
    relativeSpaceBone {boneName=root}
   }
   n38 = "Transform (Modify) Bone - Bone: dyn_clip_back_MSTR"(ComponentPose=@n37, Alpha=1, Translation="0,0,-5", Rotation="0,0,0", Scale="0,0,0") {
    rotationMode=BMM_Additive
    rotationSpace=BCS_BoneSpace
    scaleSpace=BCS_BoneSpace
    translationMode=BMM_Additive
    translationSpace=BCS_BoneSpace
    alphaBoolBlend {blendOption=Linear}
    boneToModify {boneName=dyn_clip_back_MSTR}
    scale=<0 0 0>
    translation {z=-5}
   }
   n39 = Get thigh_rot_r()
   n40 = "Transform (Modify) Bone - Bone: dyn_belt_ft_base_01"(
    ComponentPose=Local To Component(LocalPose=Component To Local(ComponentPose=@n38))
    Alpha=@n39
    Translation="0,0,0"
    Rotation="0,-20,0"
    Scale="0,0,0"
   ) {
    rotationMode=BMM_Additive
    rotationSpace=BCS_BoneSpace
    scaleSpace=BCS_BoneSpace
    translationMode=BMM_Additive
    translationSpace=BCS_BoneSpace
    alphaBoolBlend {blendOption=Linear}
    boneToModify {boneName=dyn_belt_ft_base_01}
    rotation {yaw=-20}
    scale=<0 0 0>
   }
   n41 = "Transform (Modify) Bone - Bone: dyn_belt_ft_base_00"(ComponentPose=@n40, Alpha=@n39, Translation="0,0,0", Rotation="0,-20,0", Scale="0,0,0") {
    rotationMode=BMM_Additive
    rotationSpace=BCS_BoneSpace
    scaleSpace=BCS_BoneSpace
    translationMode=BMM_Additive
    translationSpace=BCS_BoneSpace
    alphaBoolBlend {blendOption=Linear}
    boneToModify {boneName=dyn_belt_ft_base_00}
    rotation {yaw=-20}
    scale=<0 0 0>
   }
   n42 = "Constraint - dyn_belt_bk_06 (P)"(ComponentPose=@n41, Alpha=1, ConstraintWeights_0=1) {
    alphaBoolBlend {blendOption=Linear}
    boneToModify {boneName=dyn_belt_bk_06}
    constraintSetup [{targetBone {boneName=thigh_twist_01_l}}]
    constraintWeights [1]
   }
   n43 = Control Rig(Source=Component To Local(ComponentPose=@n42)) {
    alphaBoolBlend {blendOption=Linear}
    controlRigAssetReference {blueprintRigClass=&/F_MED_Poison_Hoodie_CtrlRig.F_MED_Poison_Hoodie_CtrlRig_C}
   }
   n44 = "Transform (Modify) Bone - Bone: dyn_belt_bk_06"(ComponentPose=Local To Component(LocalPose=@n43), Alpha=1, Translation="0,0,0", Rotation="0,0,0", Scale="0,0,0") {
    rotationMode=BMM_Additive
    rotationSpace=BCS_BoneSpace
    scaleSpace=BCS_BoneSpace
    translationMode=BMM_Additive
    translationSpace=BCS_BoneSpace
    alphaBoolBlend {blendOption=Linear}
    boneToModify {boneName=dyn_belt_bk_06}
    scale=<0 0 0>
   }
   n45 = "Constraint - dyn_skirt_fr_02_l (P)"(ComponentPose=@n44, Alpha=0.5, ConstraintWeights_0=1) {
    alphaBoolBlend {blendOption=Linear}
    boneToModify {boneName=dyn_skirt_fr_02_l}
    constraintSetup [{targetBone {boneName=thigh_twist_01_r}}]
    constraintWeights [1]
   }
   n46 = Get finalGravity()
   n47 = RigidBody(ComponentPose=@n45, ExternalForce=@n46) {
    bClampLinearTranslationLimitToRefPose=true
    bOverrideWorldGravity=true
    overridePhysicsAsset=&F_MED_Poison_Physics_SkirtBelt
    simulationSpace=BaseBoneSpace
    alphaBoolBlend {blendOption=Linear}
    baseBoneRef {boneName=spine_01}
    componentLinearAccScale=<0.2 0.2 0.6>
    componentLinearVelScale=<0.1 0.1 0.3>
    overrideWorldGravity {z=-1080}
   }
   n48 = RigidBody(ComponentPose=@n47, ExternalForce=@n46) {
    bClampLinearTranslationLimitToRefPose=true
    bOverrideWorldGravity=true
    overridePhysicsAsset=&F_MED_Poison_Physics_Grenade
    simulationSpace=BaseBoneSpace
    alphaBoolBlend {blendOption=Linear}
    baseBoneRef {boneName=spine_01}
    componentLinearAccScale=<0.2 0.2 0.6>
    componentLinearVelScale=<0.1 0.1 0.3>
    overrideWorldGravity {z=-1080}
   }
   n49 = Anim Dynamics - Bone: dyn_skirt_side_03_r(ComponentPose=@n48, ExternalForce="0,0,0") {
    alpha=0.6
    angularDampingOverride=3
    angularSpringConstant=50
    bAngularSpring=true
    bGravityOverrideInSimSpace=true
    bUseGravityOverride=true
    simulationSpace=BoneRelative
    alphaBoolBlend {blendOption=Linear}
    boundBone {boneName=dyn_skirt_side_03_r}
    componentAppliedLinearAccClamp=<5000 5000 5000>
    componentLinearAccScale=<0.2 0.2 0.55>
    componentLinearVelScale=<0.2 0.2 0.3>
    gravityOverride {z=-980}
    physicsBodyDefinitions [{
      sphereCollisionRadius=0
      boundBone {boneName=dyn_skirt_side_03_r}
      boxExtents=<5 25 25>
      constraintSetup {angularLimitsMax=<0 20 0>, angularLimitsMin=<-0.1 -80 -0.1>, angularTarget=<1 0 0>}
      localJointOffset=<8 0 0>
     }]
    relativeSpaceBone {boneName=root}
   }
   n50 = Anim Dynamics - Bone: dyn_skirt_fr_03_l(ComponentPose=@n49, ExternalForce="0,0,0") {
    alpha=0.6
    angularDampingOverride=3
    angularSpringConstant=50
    bAngularSpring=true
    bGravityOverrideInSimSpace=true
    bUseGravityOverride=true
    simulationSpace=BoneRelative
    alphaBoolBlend {blendOption=Linear}
    boundBone {boneName=dyn_skirt_fr_03_l}
    componentAppliedLinearAccClamp=<5000 5000 5000>
    componentLinearAccScale=<0.2 0.2 0.55>
    componentLinearVelScale=<0.2 0.2 0.3>
    gravityOverride {z=-980}
    physicsBodyDefinitions [{
      sphereCollisionRadius=0
      boundBone {boneName=dyn_skirt_fr_03_l}
      boxExtents=<5 25 25>
      constraintSetup {angularLimitsMax=<0 20 0>, angularLimitsMin=<-0.1 -60 -0.1>, angularTarget=<1 0 0>}
      localJointOffset=<8 0 0>
     }]
    relativeSpaceBone {boneName=root}
   }
   n51 = Anim Dynamics - Bone: dyn_skirt_fr_mid_r_02!(ComponentPose=@n50, ExternalForce="0,0,0") -> (Pose) {
    alpha=0.6
    angularDampingOverride=3
    angularSpringConstant=50
    bAngularSpring=true
    bGravityOverrideInSimSpace=true
    bUseGravityOverride=true
    simulationSpace=BoneRelative
    alphaBoolBlend {blendOption=Linear}
    boundBone {boneName=dyn_skirt_fr_mid_r_02}
    componentAppliedLinearAccClamp=<5000 5000 5000>
    componentLinearAccScale=<0.2 0.2 0.35>
    componentLinearVelScale=<0.1 0.1 0.2>
    gravityOverride {z=-980}
    physicsBodyDefinitions [{
      sphereCollisionRadius=0
      boundBone {boneName=dyn_skirt_fr_mid_r_02}
      boxExtents=<5 25 25>
      constraintSetup {angularLimitsMax=<0 5 0>, angularLimitsMin=<-0.1 -80 -0.1>, angularTarget=<1 0 0>}
      localJointOffset=<8 0 0>
     }]
    relativeSpaceBone {boneName=root}
   }
   n52 = Anim Dynamics - Bone: dyn_skirt_03_r!(ComponentPose=@n51, ExternalForce="0,0,0") -> (Pose) {
    alpha=0.6
    angularDampingOverride=3
    angularSpringConstant=50
    bAngularSpring=true
    bGravityOverrideInSimSpace=true
    bUseGravityOverride=true
    simulationSpace=BoneRelative
    alphaBoolBlend {blendOption=Linear}
    boundBone {boneName=dyn_skirt_03_r}
    componentAppliedLinearAccClamp=<5000 5000 5000>
    componentLinearAccScale=<0.2 0.2 0.55>
    componentLinearVelScale=<0.2 0.2 0.3>
    gravityOverride {z=-980}
    physicsBodyDefinitions [{
      sphereCollisionRadius=0
      boundBone {boneName=dyn_skirt_03_r}
      boxExtents=<5 25 25>
      constraintSetup {angularLimitsMax=<0 5 0>, angularLimitsMin=<-0.1 -80 -0.1>, angularTarget=<1 0 0>}
      localJointOffset=<8 0 0>
     }]
    relativeSpaceBone {boneName=root}
   }
   n53 = Anim Dynamics - Bone: dyn_skirt_side_03_r(ComponentPose=@n52, ExternalForce="0,0,0") {
    alpha=0.6
    angularDampingOverride=3
    angularSpringConstant=50
    bAngularSpring=true
    bGravityOverrideInSimSpace=true
    bUseGravityOverride=true
    simulationSpace=BoneRelative
    alphaBoolBlend {blendOption=Linear}
    boundBone {boneName=dyn_skirt_side_03_r}
    componentAppliedLinearAccClamp=<5000 5000 5000>
    componentLinearAccScale=<0.2 0.2 0.55>
    componentLinearVelScale=<0.2 0.2 0.3>
    gravityOverride {z=-980}
    physicsBodyDefinitions [{
      sphereCollisionRadius=0
      boundBone {boneName=dyn_skirt_side_03_r}
      boxExtents=<5 25 25>
      constraintSetup {angularLimitsMax=<0 5 0>, angularLimitsMin=<-0.1 -80 -0.1>, angularTarget=<1 0 0>}
      localJointOffset=<8 0 0>
     }]
    relativeSpaceBone {boneName=root}
   }
   n54 = Anim Dynamics - Bone: dyn_pants_l_ft_01(ComponentPose=@n53, Alpha=1, ExternalForce="0,0,0") {
    angularDampingOverride=3
    angularSpringConstant=1500
    bAngularSpring=true
    bGravityOverrideInSimSpace=true
    bUseGravityOverride=true
    simulationSpace=BoneRelative
    alphaBoolBlend {blendOption=Linear}
    boundBone {boneName=dyn_pants_l_ft_01}
    componentAppliedLinearAccClamp=<5000 5000 5000>
    componentLinearAccScale=<0.2 0.2 0.55>
    componentLinearVelScale=<0.2 0.2 0.3>
    gravityOverride {z=-1000}
    physicsBodyDefinitions [{
      sphereCollisionRadius=0
      boundBone {boneName=dyn_pants_l_ft_01}
      boxExtents=<1 20 20>
      constraintSetup {angularLimitsMax=<0 10 0>, angularLimitsMin=<-0.1 -80 -0.1>, angularTarget=<1 0 0>}
      localJointOffset=<5 0 0>
     }]
    relativeSpaceBone {boneName=root}
   }
   n55 = Anim Dynamics - Bone: dyn_pants_r_ft_01(ComponentPose=@n54, Alpha=1, ExternalForce="0,0,0") {
    angularDampingOverride=3
    angularSpringConstant=1500
    bAngularSpring=true
    bGravityOverrideInSimSpace=true
    bUseGravityOverride=true
    simulationSpace=BoneRelative
    alphaBoolBlend {blendOption=Linear}
    boundBone {boneName=dyn_pants_r_ft_01}
    componentAppliedLinearAccClamp=<5000 5000 5000>
    componentLinearAccScale=<0.2 0.2 0.55>
    componentLinearVelScale=<0.2 0.2 0.3>
    gravityOverride {z=-1000}
    physicsBodyDefinitions [{
      sphereCollisionRadius=0
      boundBone {boneName=dyn_pants_r_ft_01}
      boxExtents=<1 20 20>
      constraintSetup {angularLimitsMax=<0 10 0>, angularLimitsMin=<-0.1 -80 -0.1>, angularTarget=<1 0 0>}
      localJointOffset=<5 0 0>
     }]
    relativeSpaceBone {boneName=root}
   }
   n56 = Anim Dynamics - Bone: dyn_pants_l_in_01(ComponentPose=@n55, Alpha=1, ExternalForce="0,0,0") {
    angularDampingOverride=3
    angularSpringConstant=1500
    bAngularSpring=true
    bGravityOverrideInSimSpace=true
    bUseGravityOverride=true
    simulationSpace=BoneRelative
    alphaBoolBlend {blendOption=Linear}
    boundBone {boneName=dyn_pants_l_in_01}
    componentAppliedLinearAccClamp=<5000 5000 5000>
    componentLinearAccScale=<0.2 0.2 0.55>
    componentLinearVelScale=<0.2 0.2 0.3>
    gravityOverride {z=-1000}
    physicsBodyDefinitions [{
      sphereCollisionRadius=0
      boundBone {boneName=dyn_pants_l_in_01}
      boxExtents=<1 20 20>
      constraintSetup {angularLimitsMax=<0 10 0>, angularLimitsMin=<-0.1 -80 -0.1>, angularTarget=<1 0 0>}
      localJointOffset=<5 0 0>
     }]
    relativeSpaceBone {boneName=root}
   }
   n57 = Anim Dynamics - Bone: dyn_pants_r_in_01(ComponentPose=@n56, Alpha=1, ExternalForce="0,0,0") {
    angularDampingOverride=3
    angularSpringConstant=1500
    bAngularSpring=true
    bGravityOverrideInSimSpace=true
    bUseGravityOverride=true
    simulationSpace=BoneRelative
    alphaBoolBlend {blendOption=Linear}
    boundBone {boneName=dyn_pants_r_in_01}
    componentAppliedLinearAccClamp=<5000 5000 5000>
    componentLinearAccScale=<0.2 0.2 0.55>
    componentLinearVelScale=<0.2 0.2 0.3>
    gravityOverride {z=-1000}
    physicsBodyDefinitions [{
      sphereCollisionRadius=0
      boundBone {boneName=dyn_pants_r_in_01}
      boxExtents=<1 20 20>
      constraintSetup {angularLimitsMax=<0 10 0>, angularLimitsMin=<-0.1 -80 -0.1>, angularTarget=<1 0 0>}
      localJointOffset=<5 0 0>
     }]
    relativeSpaceBone {boneName=root}
   }
   n58 = Anim Dynamics - Bone: dyn_pants_l_ex_01(ComponentPose=@n57, Alpha=1, ExternalForce="0,0,0") {
    angularDampingOverride=3
    angularSpringConstant=1500
    bAngularSpring=true
    bGravityOverrideInSimSpace=true
    bUseGravityOverride=true
    simulationSpace=BoneRelative
    alphaBoolBlend {blendOption=Linear}
    boundBone {boneName=dyn_pants_l_ex_01}
    componentAppliedLinearAccClamp=<5000 5000 5000>
    componentLinearAccScale=<0.2 0.2 0.55>
    componentLinearVelScale=<0.2 0.2 0.3>
    gravityOverride {z=-1000}
    physicsBodyDefinitions [{
      sphereCollisionRadius=0
      boundBone {boneName=dyn_pants_l_ex_01}
      boxExtents=<1 20 20>
      constraintSetup {angularLimitsMax=<0 10 0>, angularLimitsMin=<-0.1 -80 -0.1>, angularTarget=<1 0 0>}
      localJointOffset=<5 0 0>
     }]
    relativeSpaceBone {boneName=root}
   }
   n59 = Anim Dynamics - Bone: dyn_pants_r_ex_01(ComponentPose=@n58, Alpha=1, ExternalForce="0,0,0") {
    angularDampingOverride=3
    angularSpringConstant=1500
    bAngularSpring=true
    bGravityOverrideInSimSpace=true
    bUseGravityOverride=true
    simulationSpace=BoneRelative
    alphaBoolBlend {blendOption=Linear}
    boundBone {boneName=dyn_pants_r_ex_01}
    componentAppliedLinearAccClamp=<5000 5000 5000>
    componentLinearAccScale=<0.2 0.2 0.55>
    componentLinearVelScale=<0.2 0.2 0.3>
    gravityOverride {z=-1000}
    physicsBodyDefinitions [{
      sphereCollisionRadius=0
      boundBone {boneName=dyn_pants_r_ex_01}
      boxExtents=<1 20 20>
      constraintSetup {angularLimitsMax=<0 10 0>, angularLimitsMin=<-0.1 -80 -0.1>, angularTarget=<1 0 0>}
      localJointOffset=<5 0 0>
     }]
    relativeSpaceBone {boneName=root}
   }
   n60 = Anim Dynamics - Bone: dyn_pants_l_bk_01(ComponentPose=@n59, Alpha=1, ExternalForce="0,0,0") {
    angularDampingOverride=3
    angularSpringConstant=1500
    bAngularSpring=true
    bGravityOverrideInSimSpace=true
    bUseGravityOverride=true
    simulationSpace=BoneRelative
    alphaBoolBlend {blendOption=Linear}
    boundBone {boneName=dyn_pants_l_bk_01}
    componentAppliedLinearAccClamp=<5000 5000 5000>
    componentLinearAccScale=<0.2 0.2 0.55>
    componentLinearVelScale=<0.2 0.2 0.3>
    gravityOverride {z=-1000}
    physicsBodyDefinitions [{
      sphereCollisionRadius=0
      boundBone {boneName=dyn_pants_l_bk_01}
      boxExtents=<1 20 20>
      constraintSetup {angularLimitsMax=<0 20 0>, angularLimitsMin=<-0.1 -80 -0.1>, angularTarget=<1 0 0>}
      localJointOffset=<5 0 0>
     }]
    relativeSpaceBone {boneName=root}
   }
   n61 = Anim Dynamics - Bone: dyn_pants_r_bk_01(ComponentPose=@n60, Alpha=1, ExternalForce="0,0,0") {
    angularDampingOverride=3
    angularSpringConstant=1500
    bAngularSpring=true
    bGravityOverrideInSimSpace=true
    bUseGravityOverride=true
    simulationSpace=BoneRelative
    alphaBoolBlend {blendOption=Linear}
    boundBone {boneName=dyn_pants_r_bk_01}
    componentAppliedLinearAccClamp=<5000 5000 5000>
    componentLinearAccScale=<0.2 0.2 0.55>
    componentLinearVelScale=<0.2 0.2 0.3>
    gravityOverride {z=-1000}
    physicsBodyDefinitions [{
      sphereCollisionRadius=0
      boundBone {boneName=dyn_pants_r_bk_01}
      boxExtents=<1 20 20>
      constraintSetup {angularLimitsMax=<0 20 0>, angularLimitsMin=<-0.1 -80 -0.1>, angularTarget=<1 0 0>}
      localJointOffset=<5 0 0>
     }]
    relativeSpaceBone {boneName=root}
   }
   n62 = Anim Dynamics - Bone: dyn_shoe_strap_l(ComponentPose=@n61, Alpha=1, ExternalForce="0,0,0") {
    angularDampingOverride=0.7
    angularSpringConstant=1000
    bAngularSpring=true
    bGravityOverrideInSimSpace=true
    bUseGravityOverride=true
    simulationSpace=BoneRelative
    alphaBoolBlend {blendOption=Linear}
    boundBone {boneName=dyn_shoe_strap_l}
    componentAppliedLinearAccClamp=<5000 5000 5000>
    componentLinearAccScale=<0.1 0.1 0.2>
    componentLinearVelScale=<0.05 0.05 0.1>
    gravityOverride {z=-200}
    physicsBodyDefinitions [{
      sphereCollisionRadius=0
      boundBone {boneName=dyn_shoe_strap_l}
      boxExtents=<1 20 20>
      constraintSetup {angularLimitsMax=<0 20 0>, angularLimitsMin=<-0.1 -80 -0.1>, angularTarget=<1 0 0>}
      localJointOffset=<-1 0 0>
     }]
    relativeSpaceBone {boneName=root}
   }
   n63 = Anim Dynamics - Bone: dyn_shoe_strap_r(ComponentPose=@n62, Alpha=1, ExternalForce="0,0,0") {
    angularDampingOverride=0.7
    angularSpringConstant=1000
    bAngularSpring=true
    bGravityOverrideInSimSpace=true
    bUseGravityOverride=true
    simulationSpace=BoneRelative
    alphaBoolBlend {blendOption=Linear}
    boundBone {boneName=dyn_shoe_strap_r}
    componentAppliedLinearAccClamp=<5000 5000 5000>
    componentLinearAccScale=<0.1 0.1 0.2>
    componentLinearVelScale=<0.05 0.05 0.1>
    gravityOverride {z=-200}
    physicsBodyDefinitions [{
      sphereCollisionRadius=0
      boundBone {boneName=dyn_shoe_strap_r}
      boxExtents=<1 20 20>
      constraintSetup {angularLimitsMax=<0 20 0>, angularLimitsMin=<-0.1 -80 -0.1>, angularTarget=<1 0 0>}
      localJointOffset=<-1 0 0>
     }]
    relativeSpaceBone {boneName=root}
   }
   n64 = Anim Dynamics - Bone: dyn_clip_back_01(ComponentPose=@n63, Alpha=1, ExternalForce="0,0,0") {
    angularDampingOverride=0.8
    angularSpringConstant=1000
    bAngularSpring=true
    bGravityOverrideInSimSpace=true
    bUseGravityOverride=true
    simulationSpace=BoneRelative
    alphaBoolBlend {blendOption=Linear}
    boundBone {boneName=dyn_clip_back_01}
    componentAppliedLinearAccClamp=<5000 5000 5000>
    componentLinearAccScale {z=0.8}
    componentLinearVelScale {z=0.4}
    gravityOverride {z=-1000}
    physicsBodyDefinitions [{
      sphereCollisionRadius=0
      boundBone {boneName=dyn_clip_back_01}
      boxExtents=<0.5 30 30>
      constraintSetup {angularLimitsMax=<0 0 0>, angularLimitsMin=<-0.1 -80 -0.1>, angularTarget=<1 0 0>}
      localJointOffset=<1 0 0>
     }]
    relativeSpaceBone {boneName=root}
   }
   n65 = Anim Dynamics - Bone: dyn_grenade_pin(ComponentPose=@n64, Alpha=1, ExternalForce="0,0,0") {
    angularDampingOverride=0.3
    angularSpringConstant=300
    bAngularSpring=true
    bGravityOverrideInSimSpace=true
    bUseGravityOverride=true
    simulationSpace=BoneRelative
    alphaBoolBlend {blendOption=Linear}
    boundBone {boneName=dyn_grenade_pin}
    componentLinearAccScale=<0.8 0.8 0.8>
    componentLinearVelScale=<0.4 0.4 0.4>
    gravityOverride {z=-980}
    physicsBodyDefinitions [{
      sphereCollisionRadius=0
      boundBone {boneName=dyn_grenade_pin}
      boxExtents=<7 12 12>
      constraintSetup {angularLimitsMax=<10 20 100>, angularLimitsMin=<-14 -20 0>, angularTarget=<1 0 0>}
      localJointOffset=<1 0 0>
     }]
    relativeSpaceBone {boneName=root}
   }
   n66 = Anim Dynamics - Bone: dyn_zipper_l(ComponentPose=@n65, Alpha=1, ExternalForce="0,0,0") {
    angularDampingOverride=3
    angularSpringConstant=1500
    bAngularSpring=true
    bGravityOverrideInSimSpace=true
    bUseGravityOverride=true
    simulationSpace=BoneRelative
    alphaBoolBlend {blendOption=Linear}
    boundBone {boneName=dyn_zipper_l}
    componentAppliedLinearAccClamp=<5000 5000 5000>
    componentLinearAccScale=<0.2 0.2 0.55>
    componentLinearVelScale=<0.2 0.2 0.3>
    gravityOverride {z=-1000}
    physicsBodyDefinitions [{
      sphereCollisionRadius=0
      boundBone {boneName=dyn_zipper_l}
      boxExtents=<2 20 20>
      constraintSetup {angularLimitsMax=<0 0 0>, angularLimitsMin=<-0.1 0 -80>, angularTarget=<1 0 0>}
      localJointOffset=<-5 0 0>
     }]
    relativeSpaceBone {boneName=root}
   }
   n67 = Anim Dynamics - Bone: dyn_zipper_r(ComponentPose=@n66, Alpha=1, ExternalForce="0,0,0") {
    angularDampingOverride=3
    angularSpringConstant=1500
    bAngularSpring=true
    bGravityOverrideInSimSpace=true
    bUseGravityOverride=true
    simulationSpace=BoneRelative
    alphaBoolBlend {blendOption=Linear}
    boundBone {boneName=dyn_zipper_r}
    componentAppliedLinearAccClamp=<5000 5000 5000>
    componentLinearAccScale=<0.2 0.2 0.55>
    componentLinearVelScale=<0.2 0.2 0.3>
    gravityOverride {z=-1000}
    physicsBodyDefinitions [{
      sphereCollisionRadius=0
      boundBone {boneName=dyn_zipper_r}
      boxExtents=<20 20 1>
      constraintSetup {angularLimitsMax=<0 10 0>, angularLimitsMin=<-0.1 -80 -0.1>, angularTarget=<1 0 0>}
      localJointOffset=<0 0 -5>
     }]
    relativeSpaceBone {boneName=root}
   }
   n68 = "Transform (Modify) Bone - Bone: dyn_belt_ft_base_01"(ComponentPose=@n67, Alpha=1, Translation="0.000,2,0.000", Rotation="0,0,0", Scale="0,0,0") {
    rotationMode=BMM_Additive
    translationMode=BMM_Additive
    alphaBoolBlend {blendOption=Linear}
    boneToModify {boneName=dyn_belt_ft_base_01}
    scale=<0 0 0>
    translation {y=2}
   }
   n69 = "Transform (Modify) Bone - Bone: neck_01_LOD1"(ComponentPose=@n68, Alpha=Get LOD1_selected() -> (LOD1_selected:double), Translation="0,0,0", Rotation="0,0,0", Scale="0,2.1,2") {
    rotationMode=BMM_Additive
    scaleMode=BMM_Additive
    scaleSpace=BCS_ParentBoneSpace
    translationMode=BMM_Additive
    alphaBoolBlend {blendOption=Linear}
    boneToModify {boneName=neck_01_LOD1}
    scale=<0 2.1 2>
   }
   Output Pose(Result=Component To Local(ComponentPose=@n69)) {layerGroup=None}
   Get bIsTargeting()
  ]}
 EventGraph {nodes [
   Delay(Duration=0.1)
   n1 = Get CurrentLOD()
   > Set LOD1_true(LOD1_true:double=SelectFloat(A=1, B=0, bPickA="integer >= integer"!(A:int=1, B:int=@n1) -> (ReturnValue:bool) {is=title K2Node_CallFunction})) -> (Output_Get:double)
   n2 = Get CurrentLOD()
   > Set LOD1_false(LOD1_false:double=SelectFloat(A=1, B=0, bPickA="Equal (Integer)"(A=@n2, B=0) {is=title K2Node_CallFunction})) -> (Output_Get:double)
   n3 = Delay(Duration=0.1)
   ResetDynamics(execute=[@n3.then, @n19.then_0], InTeleportType=TeleportPhysics)
   > ResetDynamics(InTeleportType=ResetPhysics)
   Event OnBeginCrouching()
   > Set Crouch(Crouch:double=0) -> (Output_Get:double)
   > Set Crouch_mask(Crouch_mask:double=1) -> (Output_Get:double)
   Event OnEndCrouching()
   > Set Crouch(Crouch=0.3)
   > Set Crouch_mask(Crouch_mask=0)
   Event BlueprintUpdateAnimation()
   n4 = "Equal (Integer)"(A=Get CurrentLOD(), B=1) {is=title K2Node_PromotableOperator}
   n5 = Get bIsPlayingEmote()
   > Set LOD1_selected(LOD1_selected:double="To Float (Boolean)"(InBool=OR Boolean(A=@n4, B=@n5))) -> (Output_Get:double)
   > n6 = Sequence!(execute) -> (then_0, then_1, then_2:Exec, then_3:Exec, then_4:Exec, then_5:Exec, then_6:Exec, then_7:Exec)
   n7 = GetOwnerSkeletalMeshComponent()
   n8 = GetDeltaTransformFromRefPose(self=@n7, BoneName=upperarm_l, BaseName=spine_05)
   Set l_armUp(execute=@n6.then_1, l_armUp:double=BreakRotator(InRot=BreakTransform(InTransform=@n8).Rotation).Yaw) -> (Output_Get:double)
   > Set l_armupClamp(l_armupClamp:double=MapRangeClamped(Value=Get l_armUp() -> (l_armUp:double), InRangeA=80, InRangeB=110)) -> (Output_Get:double)
   n9 = BreakRotator(InRot=BreakTransform(InTransform=GetDeltaTransformFromRefPose(self=@n7, BoneName=head, BaseName=neck_02)).Rotation)
   Set Head_back(execute=@n6.then_3, Head_back:double=float * float(A=MapRangeClamped(Value=@n9.Yaw, InRangeA=-8, InRangeB=-21), B=1)) -> (Output_Get:double)
   Set thigh_rot_r(
    execute=@n6.then_4
    thigh_rot_r=GetAlphaFromReferenceTransformDelta(SkeletalMeshComponent=@n7, BoneName=thigh_r, BaseName=pelvis, InputRange=(X=0,Y=-90), TwistAxis="0,0,1")
   )
   > Set thigh_rot_l(thigh_rot_l=GetAlphaFromReferenceTransformDelta(SkeletalMeshComponent=@n7, BoneName=thigh_l, BaseName=pelvis, InputRange=(X=-15,Y=-110), TwistAxis="0,0,1"))
   > n10 = Set Head Rot Up(Head Rot Up:double=GetAlphaFromReferenceTransformDelta(SkeletalMeshComponent=@n7, BoneName=head, BaseName=spine_05, InputRange=(X=-5,Y=-45), TwistAxis="0,0,1")) -> (Output_Get:double)
   > n11 = Set Head Rot L(Head Rot L:double=GetAlphaFromReferenceTransformDelta(SkeletalMeshComponent=@n7, BoneName=head, BaseName=spine_05, DeltaComponentType=EulerX, InputRange=(X=0,Y=40), TwistAxis="0,0,1")) -> (Output_Get:double)
   > Set Head Gross Rot(Head Gross Rot:double=float * float!(A=@n11, B=@n10, C:double=1.5) -> (ReturnValue)) -> (Output_Get:double)
   Set R_armupClamp2(execute=@n6.then_5, R_armupClamp2:double=MapRangeClamped(Value=Get r_armUp() -> (r_armUp:double), InRangeA=10, InRangeB=90, OutRangeB=0.8)) -> (Output_Get:double)
   n12 = Break Fort Anim Input Skydiving External Force(FortAnimInput_SkydivingExternalForce=Get SkydivingExternalForceInput()) -> (
    bUseSkydivingVectorForce=~
    SkydivingVectorForce=~
    HeadToPelvisDirection=~
    FloatingMultiplier=~
    FloatingAdditive=~
    DivingMultiplier=~
    DivingAdditive=~
    ParachutingMultiplier=~
    ParachutingAdditive=~
   )
   n13 = Set windVector(
    execute=@n6.then_6
    windVector:Vector=SelectVector(A=vector * float(A=@n12.SkydivingVectorForce, B=0.01), B="0, 0,-980", bPickA=Get bIsSkyDiving())
   ) -> (Output_Get:Vector)
   n14 = GetOwnerSkeletalMeshComponent()
   n15 = BreakRotator(InRot=BreakTransform(InTransform=GetDeltaTransformFromRefPose(self=@n14, BoneName=pelvis, BaseName=root)).Rotation)
   n16 = SelectFloat(A=980, B=-980, bPickA="float > float"(A="Absolute (Float)"(A=@n15.Yaw), B=90) {is=title K2Node_CallFunction})
   > Set finalGravity(finalGravity:Vector=SelectVector(A=MakeVector(Y=0, Z=@n16), B=@n13, bPickA=Get bIsPlayingEmote())) -> (Output_Get:Vector)
   Set Skel_Mesh(execute=@n6.then_7, Skel_Mesh=GetOwnerSkeletalMeshComponent())
   > Set thigh_rot_r_01(thigh_rot_r_01:double=GetAlphaFromReferenceTransformDelta(
     SkeletalMeshComponent=Get Skel_Mesh()
     BoneName=thigh_r
     BaseName=pelvis
     DeltaComponentType=EulerY
     InputRange=(X=5,Y=-5)
     TwistAxis="0,0,1"
    )) -> (Output_Get:double)
   n17 = GetDeltaTransformFromRefPose(self=@n7, BoneName=head, BaseName=spine_05)
   Set headFwdBk(execute=@n6.then_0, headFwdBk:double=BreakRotator(InRot=BreakTransform(InTransform=@n17).Rotation).Yaw) -> (Output_Get:double)
   > Set headRotBackClamp(headRotBackClamp:double=MapRangeClamped(Value=Get headFwdBk() -> (headFwdBk:double), InRangeA=-50, InRangeB=-20, OutRangeA=1, OutRangeB=0)) -> (Output_Get:double)
   n18 = GetDeltaTransformFromRefPose(self=@n7, BoneName=upperarm_r, BaseName=spine_05)
   Set r_armUp(execute=@n6.then_2, r_armUp:double=BreakRotator(InRot=BreakTransform(InTransform=@n18).Rotation).Yaw) -> (Output_Get:double)
   > Set r_armupClamp(r_armupClamp:double=MapRangeClamped(Value=Get r_armUp(), InRangeA=50, InRangeB=110)) -> (Output_Get:double)
   Event OnBeginJumping()
   > RetriggerableDelay(Duration=0.25)
   > Set JacketExternalForce(JacketExternalForce:Vector="0,0,1250") -> (Output_Get:Vector)
   > Set Jump(Jump:double=0) -> (Output_Get:double)
   Event OnEndJumping()
   > RetriggerableDelay(Duration=0.375)
   > Set JacketExternalForce(JacketExternalForce="0, 0, 0")
   > Set Jump(Jump=0.5)
   n19 = Sequence!(execute) -> (then_0, then_1, then_2, then_3)
   n20 = "Equal (Integer)"(A=Get CurrentLOD(), B=1) {is=title K2Node_CallFunction}
   Set LOD1_true(LOD1_true=SelectFloat(A=1, B=0, bPickA=@n20))
   > Set LOD1_front_true(LOD1_front_true:bool=@n20) -> (Output_Get:bool)
   n21 = Get CurrentLOD()
   > Set LOD1_false(LOD1_false=SelectFloat(A=1, B=0, bPickA="Equal (Integer)"(A=@n21, B=0)))
   ResetDynamics(InTeleportType=TeleportPhysics)
   SelectFloat(A=1, B=0, bPickA=false)
   Delay(Duration=1.5)
   ResetDynamics(InTeleportType=ResetPhysics)
   ResetDynamics(InTeleportType=TeleportPhysics)
   "integer >= integer"(A=1, B=0)
   Event OnCharacterCustomizationLoaded()
   Event BlueprintInitializeAnimation()
   Set Frontend_true(Frontend_true:bool=Get bIsInFrontEnd()) -> (Output_Get:bool)
  ]}
]
variables [
 Crouch {type=number}
 Crouch_mask {type=number}
 finalGravity {type=Vector}
 Frontend_true {type=boolean}
 Frontend_true_alpha {type=boolean}
 Head Gross Rot {type=number}
 Head Rot L {type=number}
 Head Rot Up {type=number}
 Head_back {type=number}
 headFwdBk {type=number}
 headRotBackClamp {type=number}
 JacketExternalForce {type=Vector}
 Jump {type=number}
 l_armUp {type=number}
 l_armupClamp {type=number}
 LOD1_false {type=number}
 LOD1_front_true {type=boolean}
 LOD1_selected {type=number}
 LOD1_true {type=number}
 r_armUp {type=number}
 r_armupClamp {type=number}
 R_armupClamp2 {type=number}
 Skel_Mesh {type=SkeletalMeshComponent}
 thigh_rot_l {type=number}
 thigh_rot_r {type=number}
 thigh_rot_r_01 {type=number}
 windVector {type=Vector}
]
