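# Physics asset for the dynamic hair bones of the F_MED_Poison female character's FaceAcc mesh part.
# Capsule bodies form six short chains (left and right bangs, front, front-left, left and right sides), each constrained back to the head with narrow swing and tighter twist limits.
# Larger head, neck_01 and spine_05 capsules supply upper-body volumes, linked by constraints with free swing and twist.
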PhysicsAsset /BRCosmetics/Characters/Player/Female/Medium/Bodies/F_MED_Poison/Meshes/Parts/F_MED_Poison_FaceAcc_Dyn
bodies {
 dyn_hair_bangs_l_00 [{shapeType=Capsule, length=2.875, radius=0.5, center=<1.6824 0 0>, rotation=<rot 0 0 57.1054>}]
 dyn_hair_bangs_l_01 [{shapeType=Capsule, length=2.875, radius=0.5, center=<1.6824 0 0>, rotation=<rot 0 0 68.3227>}]
 dyn_hair_bangs_r_00 [{shapeType=Capsule, length=2.875, radius=0.5, center=<1.6824 0 0>, rotation=<rot 0.0001 0 91.2964>}]
 dyn_hair_bangs_r_01 [{shapeType=Capsule, length=4.25, radius=1.875, center=<1.6824 0 0>, rotation=<rot 0 0 93.2446>}]
 dyn_hair_fr_00 [{shapeType=Capsule, length=3.125, radius=0.625, center=<0.5956 0 0>, rotation=<rot 0 0 90>}]
 dyn_hair_fr_01 [{shapeType=Capsule, length=2.75, radius=0.625, center=<0.4717 0 0>, rotation=<rot 0 0 90>}]
 dyn_hair_fr_02 [{shapeType=Capsule, length=2.75, radius=0.625, center=<0.8149 0 0>, rotation=<rot 0 0 90>}]
 dyn_hair_fr_03 [{shapeType=Capsule, length=2.75, radius=0.625, center=<0.6991 0 0>, rotation=<rot 0 0 90>}]
 dyn_hair_fr_l_00 [{shapeType=Capsule, length=3.375, radius=0.5, center=<1.6001 0.4916 -0.5515>, rotation=<rot -0.8183 12.6342 84.3535>}]
 dyn_hair_fr_l_01 [{shapeType=Capsule, length=3, radius=0.5, center=<1.7895 0 0>, rotation=<rot 0 0 51.1>}]
 dyn_hair_fr_l_02 [{shapeType=Capsule, length=3.375, radius=0.5, center=<1.5303 -0.0001 0.0001>, rotation=<rot 0 0 57.042>}]
 dyn_hair_l_00 [{shapeType=Capsule, length=3, radius=1.5, center=<2.3931 0 0>, rotation=<rot 0 0 90>}]
 dyn_hair_l_01 [{shapeType=Capsule, length=5.75, radius=1, center=<2.3931 0 0>, rotation=<rot 0 0 90>}]
 dyn_hair_l_02 [{shapeType=Capsule, length=5.875, radius=1, center=<2.3931 0 0>, rotation=<rot 0 0 90>}]
 dyn_hair_r_01 [{shapeType=Capsule, length=6.5, radius=0.5, center=<2.9504 0.0001 0>, rotation=<rot 0 0 90>}]
 dyn_hair_r_02 [{shapeType=Capsule, length=9.3, radius=0.5, center=<2.9504 0.0001 0>, rotation=<rot 0 0 90>}]
 dyn_hair_r_03 [{shapeType=Capsule, length=9.3, radius=0.5, center=<2.9504 0.0001 0>, rotation=<rot 0 0 90>}]
 head [{shapeType=Capsule, length=9.8473, radius=6.3232, center=<5.1387 2.9789 0.1133>, rotation=<rot 90 0 0>}]
 neck_01 [{shapeType=Capsule, length=4.0295, radius=6.5477, center=<2.1336 0.9263 0.0001>, rotation=<rot 90 0 13.5236>}]
 spine_05 [{shapeType=Capsule, length=10, radius=10, center=<2.3756 -0.0002 0>}]
}
constraints [
 {bone1Name=dyn_hair_l_01, bone2Name=dyn_hair_l_00, swing1LimitDegrees=10, swing2LimitDegrees=10, twistLimitDegrees=5}
 {bone1Name=dyn_hair_l_02, bone2Name=dyn_hair_l_01, swing1LimitDegrees=10, swing2LimitDegrees=10, twistLimitDegrees=5}
 {bone1Name=dyn_hair_r_01, bone2Name=head, swing1LimitDegrees=25, swing2LimitDegrees=25, twistLimitDegrees=5}
 {bone1Name=dyn_hair_r_02, bone2Name=dyn_hair_r_01, swing1LimitDegrees=25, swing2LimitDegrees=25, twistLimitDegrees=5}
 {bone1Name=dyn_hair_r_03, bone2Name=dyn_hair_r_02, swing1LimitDegrees=25, swing2LimitDegrees=25, twistLimitDegrees=5}
 {
  bone1Name=head
  bone2Name=neck_01
  swing1LimitDegrees=45
  swing1Motion=Free
  swing2LimitDegrees=45
  swing2Motion=Free
  twistLimitDegrees=45
  twistMotion=Free
 }
 {
  bone1Name=neck_01
  bone2Name=spine_05
  swing1LimitDegrees=45
  swing1Motion=Free
  swing2LimitDegrees=45
  swing2Motion=Free
  twistLimitDegrees=45
  twistMotion=Free
 }
 {bone1Name=dyn_hair_fr_01, bone2Name=dyn_hair_fr_00, swing1LimitDegrees=15, swing2LimitDegrees=15, twistLimitDegrees=5}
 {bone1Name=dyn_hair_fr_02, bone2Name=dyn_hair_fr_01, swing1LimitDegrees=25, swing2LimitDegrees=15, twistLimitDegrees=5}
 {bone1Name=dyn_hair_fr_00, bone2Name=head, swing1LimitDegrees=15, swing2LimitDegrees=15, twistLimitDegrees=5}
 {bone1Name=dyn_hair_fr_03, bone2Name=dyn_hair_fr_02, swing1LimitDegrees=25, swing2LimitDegrees=15, twistLimitDegrees=5}
 {bone1Name=dyn_hair_l_00, bone2Name=head, swing1LimitDegrees=10, swing2LimitDegrees=10, twistLimitDegrees=5}
 {bone1Name=dyn_hair_fr_l_02, bone2Name=dyn_hair_fr_l_01, swing1LimitDegrees=25, swing2LimitDegrees=25, twistLimitDegrees=5}
 {bone1Name=dyn_hair_fr_l_01, bone2Name=dyn_hair_fr_l_00, swing1LimitDegrees=25, swing2LimitDegrees=25, twistLimitDegrees=5}
 {bone1Name=dyn_hair_fr_l_00, bone2Name=head, swing1LimitDegrees=25, swing2LimitDegrees=25, twistLimitDegrees=5}
 {bone1Name=dyn_hair_bangs_r_00, bone2Name=head, swing1LimitDegrees=10, swing2LimitDegrees=15, twistLimitDegrees=5}
 {bone1Name=dyn_hair_bangs_r_01, bone2Name=dyn_hair_bangs_r_00, swing1LimitDegrees=10, swing2LimitDegrees=15, twistLimitDegrees=5}
 {bone1Name=dyn_hair_bangs_l_00, bone2Name=head, swing1LimitDegrees=25, swing2LimitDegrees=25, twistLimitDegrees=5}
 {bone1Name=dyn_hair_bangs_l_01, bone2Name=dyn_hair_bangs_l_00, swing1LimitDegrees=25, swing2LimitDegrees=25, twistLimitDegrees=5}
]
